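; Chainsaw weapon Blueprint that tracks the physical surface for its effects and the player pawn wielding it.
; On BeginPlay it deactivates the looping sawblade physical-surface Niagara component and caches the instigator as a Fortnite player pawn.
; SetUpCurrentSurface stores the given surface type and broadcasts it through the OnScooterSurfaceChanged dispatcher.
; GetSecondaryAbility returns only zero values.
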(event EventBeginPlay
  (|CallParentFunction)
  (Components|Activation|SetActive (Variables|Default|GetNS_SawbladeGun_Blade_loop_physSurface) false false)
  (bind _asplayer_pawn_athena (Utilities|Casting|CastToPlayerPawn_Athena (Game|GetInstigator)))
  (Variables|Default|SetPlayer _asplayer_pawn_athena))

(fn GetSecondaryAbility
  (return 0 0.0 0.0 0.0))

(fn SetUpCurrentSurface
  (params (Surface :type EPhysicalSurface%20Enum :default "SurfaceType_Default"))
  (bind _output_get (Variables|Default|SetCurrentSurface Surface))
  (Default|CallOnScooterSurfaceChanged _output_get))

(fn ConstructionScript
  (|CallParentFunction))
