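# Chainsaw reticle HUD widget that shows an RPG crosshair image, a cooldown timer and a vertical fuel bar.
# A looping timer, started on Construct and cleared on Destruct, polls the owning player's FortJumpBoostSet Fuel attribute and remaps it, clamped, into the bar's percent.
# The cooldown timer's state-change event toggles the crosshair image, hiding it during cooldown.
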WidgetBlueprint /ChainsawGameplay/Gameplay/Chainsaw_Widget
graphs [
 EventGraph {nodes [
   n1 = MoveUI() {is=title K2Node_CustomEvent}
   n2 = GetOwningPlayer()
   > SetPercent!(
    execute
    self:Progress Bar Object Reference=Get Fuel() -> (Fuel:Progress Bar Object Reference)
    InPercent:float=Remap Float(
     Input=GetFloatAttribute!(
      Actor:Actor Object Reference=@n2
      Attribute:Gameplay Attribute Structure=(AttributeName="Fuel",Attribute=/Script/FortniteGame.FortJumpBoostSet:Fuel,AttributeOwner="&Class'&FortJumpBoostSet'")
     ) -> (bSuccessfullyFoundAttribute:bool, ReturnValue:float).ReturnValue
     OldRange=(X=10,Y=100)
     NewRange=(X=0,Y=1)
     Clamp=true
    ) -> (Output=~)
   ) -> (then)
   Event Construct()
   > n3 = Set Timer by Event(Delegate=@n1, Time=0.05, bLooping=true, bMaxOncePerFrame=true)
   > Bind Event to On Cooldown State Changed(
    self:WBP Reticle Cooldown Timer Object Reference=Get WBP_Reticle_CooldownTimer_1() -> (WBP_Reticle_CooldownTimer_1:WBP Reticle Cooldown Timer Object Reference)
    Delegate=Create Event()
   ) {is=title K2Node_AddDelegate}
   > SetPercent(self=Get Fuel(), InPercent=0)
   Event Destruct()
   > Clear and Invalidate Timer by Handle(Handle=@n3)
   Event PreConstruct()
  ]}
 OnCooldownStateChanged {nodes [
   n1 = OnCooldownStateChanged!() -> (then, IsCooldownActive:bool=~) {is=title K2Node_FunctionEntry}
   > Set Is Cooldown Active(Is Cooldown Active:bool=@n1) -> (Output_Get:bool)
   > RefreshReticleVisibility()
  ]}
 RefreshReticleVisibility {nodes [
   RefreshReticleVisibility() {is=title K2Node_FunctionEntry}
   > SetVisibility(
    self=Get Image_83() -> (Image_83:Image Object Reference)
    InVisibility=Select(Option 0=Visible, Option 1=Hidden, Index=Get Is Cooldown Active() -> (Is Cooldown Active:bool))
   )
  ]}
]
variables [
 CurrentDelayedFuel {type=number, default=1}
 Is Cooldown Active {type=boolean}
 Target {type=Actor, default=null}
]
widgets {
 info {inheritedWidgetCount=0, parentClass=&UserWidget, rootWidgetClass=&CanvasPanel, widgetCount=4}
 widgets [
  CanvasPanel_29 {widgetClassPath=&CanvasPanel, widget=CanvasPanel_29}
  Image_83 {
   widgetClassPath=&Image
   bIsVariable=true
   parent=CanvasPanel_29
   widget=Image_83
   properties {
    brush {
     bIsDynamicallyLoaded=false
     drawAs=Image
     imageType=FullColor
     mirroring=NoMirror
     resourceName=None
     resourceObject=/Game/Athena/HUD/reticles/Weapons/AthenaHUD_RPG
     tiling=NoTile
     imageSize=<64 64>
     margin=<ltrb 0 0 0 0>
     outlineSettings {
      bUseBrushTransparency=false
      roundingType=HalfHeightRadius
      width=0
      color {colorUseRule=UseColor_Specified, specifiedColor=<rgba 0 0 0 0>}
      cornerRadii=<0 0 0 0>
     }
     tintColor {colorUseRule=UseColor_Specified, specifiedColor=<rgba 1 1 1 1>}
     uVRegion {bIsValid=false, max=<0 0>, min=<0 0>}
    }
    colorAndOpacity=<rgba 1 1 1 0.9>
   }
   slot {class=CanvasPanelSlot, properties {layoutData {alignment=<1 1>, anchors {maximum=<0.5 0.5>, minimum=<0.5 0.5>}, offsets=<ltrb 32 32 60 60>}}}
  }
  Fuel {
   widgetClassPath=/Script/UMG.ProgressBar
   bIsVariable=true
   parent=CanvasPanel_29
   widget=Fuel
   properties {
    barFillType=BottomToTop
    percent=0.4
    fillColorAndOpacity=<rgba 1 0.0411 0.0312 1>
    widgetStyle {
     enableFillAnimation=false
     backgroundImage {
      bIsDynamicallyLoaded=false
      drawAs=Box
      imageType=FullColor
      mirroring=NoMirror
      resourceName=../../../Engine/Content/Slate/Common/ProgressBar_Background.png
      resourceObject=null
      tiling=NoTile
      imageSize=<0 0>
      margin=<ltrb 0.4167 0.4167 0.4167 0.4167>
      outlineSettings {
       bUseBrushTransparency=false
       roundingType=HalfHeightRadius
       width=0
       color {colorUseRule=UseColor_Specified, specifiedColor=<rgba 0 0 0 0>}
       cornerRadii=<0 0 0 0>
      }
      tintColor {colorUseRule=UseColor_Specified, specifiedColor=<rgba 0 0 0 0.419>}
      uVRegion {bIsValid=false, max=<0 0>, min=<0 0>}
     }
     fillImage {
      bIsDynamicallyLoaded=false
      drawAs=Box
      imageType=FullColor
      mirroring=NoMirror
      resourceName=../../../Engine/Content/Slate/Common/ProgressBar_Fill.png
      resourceObject=null
      tiling=NoTile
      imageSize=<0 0>
      margin=<ltrb 0.4167 0.4167 0.4167 0.4167>
      outlineSettings {
       bUseBrushTransparency=false
       roundingType=HalfHeightRadius
       width=0
       color {colorUseRule=UseColor_Specified, specifiedColor=<rgba 0 0 0 0>}
       cornerRadii=<0 0 0 0>
      }
      tintColor {colorUseRule=UseColor_Specified, specifiedColor=<rgba 1 1 1 1>}
      uVRegion {bIsValid=false, max=<0 0>, min=<0 0>}
     }
     marqueeImage {
      bIsDynamicallyLoaded=false
      drawAs=Image
      imageType=FullColor
      mirroring=NoMirror
      resourceName=../../../Engine/Content/Slate/Common/ProgressBar_Marquee.png
      resourceObject=null
      tiling=Horizontal
      imageSize=<20 12>
      margin=<ltrb 0 0 0 0>
      outlineSettings {
       bUseBrushTransparency=false
       roundingType=HalfHeightRadius
       width=0
       color {colorUseRule=UseColor_Specified, specifiedColor=<rgba 0 0 0 0>}
       cornerRadii=<0 0 0 0>
      }
      tintColor {colorUseRule=UseColor_Specified, specifiedColor=<rgba 1 1 1 1>}
      uVRegion {bIsValid=false, max=<0 0>, min=<0 0>}
     }
    }
   }
   slot {
    class=CanvasPanelSlot
    properties {zOrder=1, layoutData {alignment=<0 0>, anchors {maximum=<0.5 0.5>, minimum=<0.5 0.5>}, offsets=<ltrb 63.0391 -32.5405 10 70>}}
   }
  }
  WBP_Reticle_CooldownTimer_1 {
   widgetClassPath=/Game/Athena/HUD/reticles/Widgets/WBP_Reticle_CooldownTimer
   bIsVariable=true
   parent=CanvasPanel_29
   widget=WBP_Reticle_CooldownTimer_1
   slot {
    class=CanvasPanelSlot
    properties {layoutData {alignment=<0 0>, anchors {maximum=<0.5 0.5>, minimum=<0.5 0.5>}, offsets=<ltrb -30 -30 60 60>}}
   }
  }
 ]
}
