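; Gameplay ability for the chainsaw melee attack, extending a parent Mantis technique ability.
; Technique start applies a sound-indicator effect to the user, hits attach chainsaw rumble feedback to the pawn mesh, and ending clears the aim-assist timer.
; Targeting originates from the player camera transform with a downward offset.
; Activation failures blocked by primary-attack, combo or Mantis-attack tags return a no-message reason, and debug prints mark activation and end.
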(event Mantis|EventOnMantisTechniqueHit
  (params (TargetDataHandle :type Gameplay%20Ability%20Target%20Data%20Handle%20Structure%20%28by%20ref%29) (ApplicationTag :type Gameplay%20Tag%20Structure))
  (|CallParentFunction TargetDataHandle ApplicationTag)
  (ForceFeedback|SpawnForceFeedbackAttached "/ChainsawGameplay/Gameplay/CamShakeAndRumble/FF_HitSomething_Chainsaw.FF_HitSomething_Chainsaw" (Variables|/Script/Engine.Character|GetMesh (Ability|GetActivatingPawn))))

(event Ability|EventActivateAbility
  (|CallParentFunction)
  (Development|PrintString "MANTIS ACTIVATED"))

(event Mantis|EventOnMantisAblilityTechniqueStarted
  (params (OwningTags :type Gameplay%20Tag%20Container%20Structure) (SequenceName :type Name :default "None") (StartsSequence :type Boolean) (EndsSequence :type Boolean))
  (|CallParentFunction OwningTags SequenceName StartsSequence EndsSequence)
  (GameplayEffects|ApplyGameplayEffectToSelf :self (Ability|GetAbilitySystemComponent (Ability|GetActivatingPawn)) :GameplayEffectClass "/ChainsawGameplay/Gameplay/GameplayEffects/GE_ApplySoundIndicatorCue.GE_ApplySoundIndicatorCue_C" :Level 0.0))

(event Ability|EventOnEndAbility
  (params (bWasCancelled :type Boolean))
  (Utilities|Time|ClearandInvalidateTimerbyHandle (Variables|Default|GetAimAssistTargetingTimerHandle))
  (Development|PrintString "Scythe ability ernd" false))

(fn GetCustomAbilitySourceTransform
  (bind _asfort_pawn (Utilities|Casting|CastToFortPawn (Ability|GetAvatarActorfromActorInfo)))
  (bind _returnvalue (TargetingSystem|GetPlayerCameraTargetingTransform _asfort_pawn true "0.000000,0.000000,-25.000000"))
  (return _returnvalue))

(fn OverrideFailedReason
  (params (FailedReason :type Gameplay%20Tag%20Container%20Structure%20%28by%20ref%29))
  (if (Math|Boolean|ORBoolean (GameplayTags|HasAllTags FailedReason (GameplayTags|MakeLiteralGameplayTagContainer "(GameplayTags=((TagName=\"Activate.Failed.Tags.Blocked\"),(TagName=\"Granted.Athena.Chainsaw.BlockPrimaryAttack\")))") true) (GameplayTags|HasAllTags FailedReason (GameplayTags|MakeLiteralGameplayTagContainer "(GameplayTags=((TagName=\"Activate.Failed.Tags.Blocked\"),(TagName=\"Gameplay.Action.Player.Chainsaw.Combo\")))") true) (GameplayTags|HasAllTags FailedReason (GameplayTags|MakeLiteralGameplayTagContainer "(GameplayTags=((TagName=\"Activate.Failed.Tags.Blocked\"),(TagName=\"Gameplay.Action.Player.Mantis.Attack\")))") true))
    (return "(GameplayTags=((TagName=\"Athena.Ability.Failed.NoMessage\")))")
    (else
      (return FailedReason))))
