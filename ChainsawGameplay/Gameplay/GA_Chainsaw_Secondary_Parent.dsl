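; Parent gameplay ability for chainsaw secondary attacks, providing shared lifecycle, damage and feedback logic.
; Activation caches the chainsaw and shows reticle status, then ends or cancels on tactical sprint, a cancel event or a safety timeout.
; Ending stops the montage and clears the reticle.
; Helpers apply the damage effect after a team check that lets player-built structures and friendly-fire minigames through, and play hit cues, camera shake and rumble.
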(event Custom|OnEffectsAppliedToTargets
  (params (TargetPawn :type Array%20of%20Actor%20Object%20References) (TargetEnvironment :type Array%20of%20Actor%20Object%20References))
  ;[1 1 1] Gameplay Cues for enemy and environment hits
  (for _array_element TargetPawn
    (Ability|ExecuteGameplayCueWithParamsOnOwner (Variables|Default|GetGCN_EnemyHit) (Ability|GameplayCue|MakeGameplayCueParameters :NormalizedMagnitude 0.0 :RawMagnitude 0.0 :Location (Transformation|GetActorLocation _array_element) :Normal "0,0,0" :GameplayEffectLevel 1 :AbilityLevel 1 :bReplicateLocationWhenUsingMinimalRepProxy false)))
  (for _array_element_2 TargetEnvironment
    (Ability|ExecuteGameplayCueWithParamsOnOwner (Variables|Default|GetGCN_EnvironmentHit) (Ability|GameplayCue|MakeGameplayCueParameters :NormalizedMagnitude 0.0 :RawMagnitude 0.0 :Location (Transformation|GetActorLocation _array_element_2) :Normal "0,0,0" :GameplayEffectLevel 1 :AbilityLevel 1 :bReplicateLocationWhenUsingMinimalRepProxy false))))

(event Ability|EventActivateAbility
  (bind (_isvalid _weapon) (CallFunction|GetWeapon))
  (if _isvalid
    (Variables|Default|SetWeapon _weapon)
    (Class|BMantisWeaponParent|ReticleStatusDisplay (Variables|Default|GetWeapon) true)
    (Ability|Tasks|WaitGameplayEvent (Variables|Default|GetSecondary%20Cancel%20Tag)
      (:then
        ;[1 1 1] FORT-454406
        (Ability|Tasks|WaitGameplayTagAdd "(TagName=\"Gameplay.Action.Player.TacticalSprint\")" (Ability|GetAvatarActorfromActorInfo)
          (:then
            (Utilities|FlowControl|Delay (Variables|Default|GetForce%20End%20Delay)) ; Safety cancel
            (CallFunction|OnAbilityCancelled))
          (:Added
            (Ability|EndAbility))))
      (:EventReceived
        (CallFunction|OnAbilityCancelled)))
    (else
      (Ability|EndAbility))))

(event Custom|On%20Ability%20Cancelled
  (Ability|CancelAbility))

(event Ability|EventOnEndAbility
  (params (bWasCancelled :type Boolean))
  (CallFunction|ForceEndMontages)
  (bind _weapon (Variables|Default|GetWeapon)
    (:then
      (Class|BMantisWeaponParent|ReticleStatusDisplay _weapon false))
    (:else)))

(fn ApplyEffectsToTargets
  (params (TargetData :type Gameplay%20Ability%20Target%20Data%20Handle%20Structure) (JumpAttack :type Boolean))
  ;[0 0.8002 1 0.25] Applies a Gameplay Effect to any targets using data from the source weapon to determine damage
  (bind _onsameteam (CallFunction|FriendlyFireCheck TargetData))
  (if (not _onsameteam)
    (Ability|ApplyGameplayEffectToTarget (Variables|GetTargetData) (Variables|Default|GetGE%20Class))))

(fn ForceEndMontages
  (bind _montageinfo_montagetoplay (Utilities|Struct|BreakFortGameplayAbilityMontageInfo (Variables|GetMontageInfo_MontageToPlay)))
  (bind _outfortplayerpawn (CallFunction|GetAvatarasValidFortPlayerPawn)
    (:ValidFortPlayerPawn
      (Animation|StopAnimMontage _outfortplayerpawn _montageinfo_montagetoplay))))

(fn FriendlyFireCheck
  (params (TargetData :type Gameplay%20Ability%20Target%20Data%20Handle%20Structure))
  (bind _returnvalue (Ability|GetContextfromOwner TargetData))
  (bind (_returnvalue_bblockinghit _returnvalue_binitialoverlap _returnvalue_time _returnvalue_distance _returnvalue_location _returnvalue_impactpoint _returnvalue_normal _returnvalue_impactnormal _returnvalue_physmat _returnvalue_hitactor) (Collision|BreakHitResult (Ability|EffectContext|GetHitResult _returnvalue)))
  (bind _returnvalue_2 (Ability|GetAbilitySystemComponent _returnvalue_hitactor))
  (bind _outfortplayerpawn (CallFunction|GetAvatarasValidFortPlayerPawn)
    (:ValidFortPlayerPawn
      (switch Utilities|FlowControl|Switch|SwitchonEFortTeamAffiliation (Team|GetTeamAffiliation _outfortplayerpawn _returnvalue_hitactor)
        (:Friendly
          (Utilities|IsValid (Minigame|GetMinigame _outfortplayerpawn :!class /Script/FortniteGame.MinigameBlueprintLibrary)
            (:Is%20Valid
              (if (Variables|/Script/FortniteGame.FortMinigameGameRulesComponent|GetbAllowFriendlyFire (Variables|/Script/FortniteGame.FortMinigame|GetGameRulesComponent (Minigame|GetMinigame _outfortplayerpawn :!class /Script/FortniteGame.MinigameBlueprintLibrary)))
                (return false)
                (else
                  ;[1 1 1] let player built walls be on a "different team" so they can take damage. this mimmics normal behavior of players being ablt to damage their own walls with other weapons but in this case, for secondary abilities
                  (Utilities|IsValid _returnvalue_2
                    (:Is%20Valid
                      (return (not (GameplayTags|HasMatchingGameplayTag _returnvalue_2 "(TagName=\"Building.Type.Player\")"))))
                    (:Is%20Not%20Valid
                      (return true))))))
            (:Is%20Not%20Valid
              (return true))))
        (:Neutral
          _)
        (:Hostile
          _))
      (return false))
    (:AvatarCastFailed
      (return false))))

(fn GetWeapon
  (bind _outfortplayerpawn (CallFunction|GetAvatarasValidFortPlayerPawn)
    (:ValidFortPlayerPawn
      (bind _asb_weapon_chainsaw (Utilities|Casting|CastToB_Weapon_Chainsaw (Weapon|GetCurrentWeapon _outfortplayerpawn))
        (:then
          (return true _asb_weapon_chainsaw))
        (:CastFailed
          _)))
    (:AvatarCastFailed
      _))
  (return false 0))

(fn PlayForceFeedbackAndCameraShake
  (params (ShakeClass :type Legacy%20Camera%20Shake%20Class%20Reference) (ForceFeedbackEffect :type Force%20Feedback%20Effect%20Object%20Reference))
  (bind _outfortplayerpawn (CallFunction|GetAvatarasValidFortPlayerPawn)
    (:ValidFortPlayerPawn
      (CameraShakes|StartCameraShake :self (Variables|/Script/Engine.PlayerController|GetPlayerCameraManager (Fortnite|GetFortPlayerControllerfromActor _outfortplayerpawn)) :ShakeClass (Variables|GetShakeClass) :Scale 1.0 :PlaySpace "CameraLocal")
      (Fortnite|PlayLocalForceFeedbackatLocation (Variables|GetForceFeedbackEffect) (Transformation|GetActorLocation _outfortplayerpawn) 1024.0))))
