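; Gameplay ability that launches the player for the Chainsaw jump attack, built on a parent knockback/impulse ability.
; On activation it runs async aim-assist targeting, aiming the impulse at the first hit actor or, without one, deriving it from current velocity.
; A scooter-activation flag changes the no-target speed and added upward boost before the parent activates.
; Overrides supply that velocity and a camera-relative knockback origin, with optional debug drawing.
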(event Ability|EventActivateAbility
  (bind _outfortplayerpawn (CallFunction|GetAvatarasValidFortPlayerPawn)
    (:ValidFortPlayerPawn
      (Variables|Default|SetPlayer _outfortplayerpawn)
      (bind (_asynctaskproxy _targetinghandle) (Class|AsyncActionPerformTargeting|PerformTargetingAsyncAction (Variables|Default|GetPlayer) (Variables|Default|GetTP_AimAssist))
        (:Targeted
          (if (Utilities|Array|IsEmpty (TargetingSystem|TargetingResults|GetTargetingResults "/Script/TargetingSystem.Default__TargetingSubsystem" _targetinghandle))
            (Variables|Default|SetImpulseVelocity (Math|Vector|SelectVector (* (Math|Vector|Normalize2D(Vector) (Transformation|GetVelocity (Variables|Default|GetPlayer))) 1.5e+03) (Math|Vector|VectorClampSizeMax (Transformation|GetVelocity (Variables|Default|GetPlayer)) (Variables|/Script/Engine.CharacterMovementComponent|GetMaxWalkSpeed (Variables|/Script/Engine.Character|GetCharacterMovement (Variables|Default|GetPlayer)))) (Variables|Default|GetActivatedFromScooter)))
            (else
              (CallFunction|DrawDebug (Transformation|GetActorLocation (.HitActor (Collision|BreakHitResult (Utilities|Array|Get(aref) (TargetingSystem|TargetingResults|GetTargetingResults "/Script/TargetingSystem.Default__TargetingSubsystem" _targetinghandle))))))
              (Variables|Default|SetImpulseVelocity (* (- (Transformation|GetActorLocation (.HitActor (Collision|BreakHitResult (Utilities|Array|Get(aref) (TargetingSystem|TargetingResults|GetTargetingResults "/Script/TargetingSystem.Default__TargetingSubsystem" _targetinghandle))))) (Transformation|GetActorLocation (Variables|Default|GetPlayer))) (Math|Float|SelectFloat 1.0 1.0 (Variables|Default|GetActivatedFromScooter))))))
          (Variables|SetImpulseVelocity_X (Math|Vector|MakeVector (.x (Variables|Default|GetImpulseVelocity)) (.y (Variables|Default|GetImpulseVelocity)) (+ (.z (Variables|Default|GetImpulseVelocity)) (Math|Float|SelectFloat 7e+02 5.5e+02 (Variables|Default|GetActivatedFromScooter)))))
          (|CallParentFunction))))
    (:AvatarCastFailed
      (Ability|EndAbility))))

(fn CalculateVelocityMaintained
  (params (LaunchTarget :type Actor%20Object%20Reference) (LaunchMagIn :type Vector :default "0, 0, 0"))
  (return (Variables|Default|GetImpulseVelocity)))

(fn DrawDebug
  (params (Center :type Vector))
  (if (Variables|Default|GetEnableDebugStuff)
    (Rendering|Debug|DrawDebugSphere Center 1e+02 12 "(R=1.000000,G=0.026704,B=0.000000,A=1.000000)" 2.0 5.0)))

(fn GetKnockbackOrigin
  (params (OriginActor :type Actor%20Object%20Reference))
  (return (- (Transformation|GetActorLocation (Variables|KnockbackParent|GetTarget)) (Math|Vector|Normalize2D(Vector) (Math|Vector|GetForwardVector (Camera|GetCameraRotation (Variables|/Script/Engine.PlayerController|GetPlayerCameraManager (Fortnite|GetFortPlayerControllerfromActor OriginActor))))))))

(fn GetTargetLocation
  (bind _returnvalue (|CallParentFunction))
  (return _returnvalue))
