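; Chainsaw jump attack gameplay ability that launches the player toward an aim-assist target, or along clamped current velocity, then strikes on landing.
; A looping fall montage and cues run until a movement-mode task or landed delegate detects ground.
; Landing plays a montage and slam cue, runs async targeting, applies effects and spends weapon durability.
; Swimming, skydiving or clambering cancels it and suppresses the slam cue; DoOnce gates reset each activation.
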(event Custom|ListenForSwimming
  ;[1 1 1] Listen for water
  (bind _asynctaskproxy (Ability|Tasks|WaitGameplayTagAdd "(TagName=\"MovementMode.SurfaceSwimming\")" (Ability|GetAvatarActorfromActorInfo))
    (:then
      (Variables|Runtime|SetSurfaceSwimmingAsyncTask _asynctaskproxy)
      ;[1 1 1] listen for parachuting
      (bind _asynctaskproxy_2 (Ability|Tasks|WaitGameplayTagAdd "(TagName=\"MovementMode.Skydiving\")" (Ability|GetAvatarActorfromActorInfo))
        (:then
          (Variables|Runtime|SetSkydivingAsyncTask _asynctaskproxy_2))
        (:Added
          (Variables|Default|SetbIgnoreLanded true)
          (Ability|CancelAbility))))
    (:Added
      (Variables|Default|SetbIgnoreLanded true)
      (Ability|CancelAbility))))

(event Custom|Predict
  ;[1 1 1 1 64] PredictionWindow
  (EventDispatchers|AssignOnPawnLanded (Variables|Runtime|GetOut%20Fort%20Player%20Pawn) (delegate Custom|FallbackLandEvent))
  (CallFunction|WaitforLand))

(event Custom|FallbackLandEvent
  (params (Hit :type Hit%20Result%20Structure%20%28by%20ref%29))
  ;[1 1 1 1 64] PredictionWindow
  (EventDispatchers|UnbindEventfromOnPawnLanded (Variables|Runtime|GetOut%20Fort%20Player%20Pawn) (delegate Custom|FallbackLandEvent))
  (Utilities|Time|ClearandInvalidateTimerbyHandle (Variables|Runtime|GetPredictLandingTimerHandle))
  (CallFunction|PredictionHit))

(event Custom|UnbindFallbackLandCheck
  ;[1 1 1 1 64] PredictionWindow
  (EventDispatchers|UnbindEventfromOnPawnLanded (Variables|Runtime|GetOut%20Fort%20Player%20Pawn) (delegate Custom|FallbackLandEvent))
  (Utilities|Time|ClearandInvalidateTimerbyHandle (Variables|Runtime|GetPredictLandingTimerHandle))
  (CallFunction|PredictionHit))

(event Custom|PredictionHit
  ;[1 1 1 1 64] Prediction Complete
  (Utilities|FlowControl|DoOnce false
    (:Completed
      (if (not (Variables|Default|GetHasLanded))
        _))))

(event Custom|ResetPredictionHit
  ;[1 1 1 1 64] Prediction Complete
  (Utilities|FlowControl|DoOnce false
    (:Completed
      (if (not (Variables|Default|GetHasLanded))
        _))))

(event Custom|WaitForLand
  ;[1 1 1 1 64] On Landed
  (Utilities|FlowControl|DoOnce false
    (:Completed
      (bind _asynctaskproxy (Ability|Tasks|WaitMovementModeChange "MOVE_Walking")
        (:then
          (Variables|Runtime|SetWaitForLandAsyncTask _asynctaskproxy)
          (EventDispatchers|AssignOnPawnLanded (Variables|Runtime|GetOut%20Fort%20Player%20Pawn) (delegate Custom|AbilityPlayerPawnLanded))
          (Ability|ExecuteGameplayCueOnOwner (Variables|Const|GetJumpAttackPredictLandTag) (Class|AthenaFunctionLibrary|MakeSimpleFortEffectContext (Ability|GetAvatarActorfromActorInfo) (Ability|GetAvatarActorfromActorInfo))))
        (:OnChange
          (if (Variables|Default|GetbCanPerformHit)
            (CallFunction|UnbindAbilityPlayerPawnLanded)
            (CallFunction|PlayAbilityPlayerPawnLandedEvent)))))))

(event Custom|ResetWaitForLand
  ;[1 1 1 1 64] On Landed
  (Utilities|FlowControl|DoOnce false
    (:Completed
      (bind _asynctaskproxy (Ability|Tasks|WaitMovementModeChange "MOVE_Walking")
        (:then
          (Variables|Runtime|SetWaitForLandAsyncTask _asynctaskproxy)
          (EventDispatchers|AssignOnPawnLanded (Variables|Runtime|GetOut%20Fort%20Player%20Pawn) (delegate Custom|AbilityPlayerPawnLanded))
          (Ability|ExecuteGameplayCueOnOwner (Variables|Const|GetJumpAttackPredictLandTag) (Class|AthenaFunctionLibrary|MakeSimpleFortEffectContext (Ability|GetAvatarActorfromActorInfo) (Ability|GetAvatarActorfromActorInfo))))
        (:OnChange
          (if (Variables|Default|GetbCanPerformHit)
            (CallFunction|UnbindAbilityPlayerPawnLanded)
            (CallFunction|PlayAbilityPlayerPawnLandedEvent)))))))

(event Custom|UnbindAbilityPlayerPawnLanded
  ;[1 1 1 1 64] On Landed
  (EventDispatchers|UnbindEventfromOnPawnLanded (Variables|Runtime|GetOut%20Fort%20Player%20Pawn) (delegate Custom|AbilityPlayerPawnLanded)))

(event Custom|AbilityPlayerPawnLanded
  (params (Hit :type Hit%20Result%20Structure%20%28by%20ref%29))
  ;[1 1 1 1 64] On Landed
  (if (Variables|Default|GetbCanPerformHit)
    (GameplayTasks|EndTask (Variables|Runtime|GetWaitForLandAsyncTask))
    (CallFunction|UnbindAbilityPlayerPawnLanded)
    (CallFunction|PlayAbilityPlayerPawnLandedEvent)))

(event Ability|EventActivateAbility
  ;[1 1 1 1 64] Ability
  (if (ScalableFloat|GetValueasBool (Variables|Default|GetJumpAttackEnabledHotfixRow))
    (CallFunction|GetAvatarasValidFortPlayerPawn
      (:ValidFortPlayerPawn
        (CallFunction|UpdateDelayTimefromGravity)
        (bind _outfortplayerpawn (CallFunction|GetAvatarasValidFortPlayerPawn)
          (:ValidFortPlayerPawn
            (|CallParentFunction)
            ;;[0.004746 0.07299 0.15 0.5] parent ability may have ended this ability so check for that and exit out, else we could end up playing montage after ability is over (FORT-793602).
            (if (or (Ability|GameplayAbility|IsActive 0) (not (ScalableFloat|GetValueasBool (Variables|Default|GetCancelActivationIfParentEndsAbility))))
              (Pawn|BindEventtoOnPawnJumped _outfortplayerpawn (delegate Custom|OnJumpedAgain))
              (Variables|Default|SetbCanPerformHit true)
              (Variables|Default|SetHasLanded false)
              (Variables|Default|SetbIgnoreLanded false)
              (CallFunction|ListenforSwimming)
              (CallFunction|CancelifClambering)
              (CallFunction|ActivateLoopingWeaponFX)
              (GameplayCue|ExecuteGameplayCue%28Burst%29Local (Variables|Default|GetFortPlayerpawn_0) (Variables|Default|GetGC_JumpStarted) (Ability|GameplayCue|MakeGameplayCueParameters 0.0))
              (CallFunction|Reset)
              (Utilities|FlowControl|DoOnce false
                (:Completed
                  (bind _outfortplayerpawn_2 (CallFunction|GetAvatarasValidFortPlayerPawn)
                    (:ValidFortPlayerPawn
                      (Variables|Runtime|SetOut%20Fort%20Player%20Pawn _outfortplayerpawn_2)
                      (bind (_asynctaskproxy _targetinghandle) (Class|AsyncActionPerformTargeting|PerformTargetingAsyncAction (Variables|Runtime|GetOut%20Fort%20Player%20Pawn) (Variables|Const|GetTP_AimAssist))
                        (:Targeted
                          (if (Utilities|Array|IsEmpty (TargetingSystem|TargetingResults|GetTargetingResults "/Script/TargetingSystem.Default__TargetingSubsystem" _targetinghandle))
                            (Variables|Default|SetImpulseDirection (Math|Vector|VectorClampSizeMax (Transformation|GetVelocity (Variables|Runtime|GetOut%20Fort%20Player%20Pawn)) (Variables|/Script/Engine.CharacterMovementComponent|GetMaxWalkSpeed (Variables|/Script/Engine.Character|GetCharacterMovement (Variables|Runtime|GetOut%20Fort%20Player%20Pawn)))))
                            (else
                              (CallFunction|DrawDebug (Transformation|GetActorLocation (.HitActor (Collision|BreakHitResult (Utilities|Array|Get(aref) (TargetingSystem|TargetingResults|GetTargetingResults "/Script/TargetingSystem.Default__TargetingSubsystem" _targetinghandle))))))
                              (Variables|Default|SetImpulseDirection (- (Transformation|GetActorLocation (.HitActor (Collision|BreakHitResult (Utilities|Array|Get(aref) (TargetingSystem|TargetingResults|GetTargetingResults "/Script/TargetingSystem.Default__TargetingSubsystem" _targetinghandle))))) (Transformation|GetActorLocation (Variables|Runtime|GetOut%20Fort%20Player%20Pawn))))))
                          (Pawn|Character|LaunchCharacterJump (Variables|Runtime|GetOut%20Fort%20Player%20Pawn) (Math|Vector|MakeVector (.x (Variables|Default|GetImpulseDirection)) (.y (Variables|Default|GetImpulseDirection)) (+ (.z (Variables|Default|GetImpulseDirection)) 6.5e+02)) true true true true)
                          (Ability|Tasks|PlayMontageAndWait "None" (Variables|AnimationSettings|GetJumpAndFallLoopMontage))
                          (CallFunction|Predict))))
                    (:AvatarCastFailed
                      (Ability|EndAbility)))))))))
      (:AvatarCastFailed
        _))
    (else
      _))
  (Ability|EndAbility))

(event Custom|ResetActivation
  ;[1 1 1 1 64] Ability
  (Utilities|FlowControl|DoOnce false
    (:Completed
      (bind _outfortplayerpawn (CallFunction|GetAvatarasValidFortPlayerPawn)
        (:ValidFortPlayerPawn
          (Variables|Runtime|SetOut%20Fort%20Player%20Pawn _outfortplayerpawn)
          (bind (_asynctaskproxy _targetinghandle) (Class|AsyncActionPerformTargeting|PerformTargetingAsyncAction (Variables|Runtime|GetOut%20Fort%20Player%20Pawn) (Variables|Const|GetTP_AimAssist))
            (:Targeted
              (if (Utilities|Array|IsEmpty (TargetingSystem|TargetingResults|GetTargetingResults "/Script/TargetingSystem.Default__TargetingSubsystem" _targetinghandle))
                (Variables|Default|SetImpulseDirection (Math|Vector|VectorClampSizeMax (Transformation|GetVelocity (Variables|Runtime|GetOut%20Fort%20Player%20Pawn)) (Variables|/Script/Engine.CharacterMovementComponent|GetMaxWalkSpeed (Variables|/Script/Engine.Character|GetCharacterMovement (Variables|Runtime|GetOut%20Fort%20Player%20Pawn)))))
                (else
                  (CallFunction|DrawDebug (Transformation|GetActorLocation (.HitActor (Collision|BreakHitResult (Utilities|Array|Get(aref) (TargetingSystem|TargetingResults|GetTargetingResults "/Script/TargetingSystem.Default__TargetingSubsystem" _targetinghandle))))))
                  (Variables|Default|SetImpulseDirection (- (Transformation|GetActorLocation (.HitActor (Collision|BreakHitResult (Utilities|Array|Get(aref) (TargetingSystem|TargetingResults|GetTargetingResults "/Script/TargetingSystem.Default__TargetingSubsystem" _targetinghandle))))) (Transformation|GetActorLocation (Variables|Runtime|GetOut%20Fort%20Player%20Pawn))))))
              (Pawn|Character|LaunchCharacterJump (Variables|Runtime|GetOut%20Fort%20Player%20Pawn) (Math|Vector|MakeVector (.x (Variables|Default|GetImpulseDirection)) (.y (Variables|Default|GetImpulseDirection)) (+ (.z (Variables|Default|GetImpulseDirection)) 6.5e+02)) true true true true)
              (Ability|Tasks|PlayMontageAndWait "None" (Variables|AnimationSettings|GetJumpAndFallLoopMontage))
              (CallFunction|Predict))))
        (:AvatarCastFailed
          (Ability|EndAbility))))))

(event Custom|ForceLanding
  ;[1 1 1 1 64] On Landed
  (if (Variables|Default|GetbCanPerformHit)
    (CallFunction|UnbindAbilityPlayerPawnLanded)
    (CallFunction|PlayAbilityPlayerPawnLandedEvent)))

(event Custom|OnJumpedAgain
  ;[1 1 1 1 64] Ability
  (Ability|EndAbility))

(event Custom|PlayAbilityPlayerPawnLandedEvent
  ;[1 1 1 1 64] On Landed
  (Utilities|FlowControl|DoOnce false
    (:Completed
      (CallFunction|DeactivateLoopingWeaponFX)
      (bind (_asynctaskproxy _notifytag) (Ability|Tasks|PlayMontageAndWaitNotify "None" (Variables|AnimationSettings|GetLandedOnGroundMontage) 1.0 "None" false)
        (:then
          (Variables|Default|SetHasLanded true)
          ;;[0 1 0.1364] This finishing on the client first can cause potential issues if the player swings the weapon before the server says its ok
          (CallFunction|DoTargetting)
          (CallFunction|OnPawnLanded))
        (:OnCompleted
          _)
        (:OnInterrupted
          _)
        (:OnCancelled
          _)
        (:OnNotifyBegin
          (if (== _notifytag (Variables|Const|GetRemoveAttackSuppressionTag))
            (Ability|EndAbility))))
      (Ability|Tasks|WaitNetSync "BothWait"
        (:OnSync
          (Ability|EndAbility))))))

(event Custom|ResetPlayAblityPlayerPawnLandedEvent
  ;[1 1 1 1 64] On Landed
  (Utilities|FlowControl|DoOnce false
    (:Completed
      (CallFunction|DeactivateLoopingWeaponFX)
      (bind (_asynctaskproxy _notifytag) (Ability|Tasks|PlayMontageAndWaitNotify "None" (Variables|AnimationSettings|GetLandedOnGroundMontage) 1.0 "None" false)
        (:then
          (Variables|Default|SetHasLanded true)
          ;;[0 1 0.1364] This finishing on the client first can cause potential issues if the player swings the weapon before the server says its ok
          (CallFunction|DoTargetting)
          (CallFunction|OnPawnLanded))
        (:OnCompleted
          _)
        (:OnInterrupted
          _)
        (:OnCancelled
          _)
        (:OnNotifyBegin
          (if (== _notifytag (Variables|Const|GetRemoveAttackSuppressionTag))
            (Ability|EndAbility))))
      (Ability|Tasks|WaitNetSync "BothWait"
        (:OnSync
          (Ability|EndAbility))))))

(event Custom|ForceLandedEvent
  ;[1 1 1 1 64] On Landed
  (CallFunction|DeactivateLoopingWeaponFX)
  (bind (_asynctaskproxy _notifytag) (Ability|Tasks|PlayMontageAndWaitNotify "None" (Variables|AnimationSettings|GetLandedOnGroundMontage) 1.0 "None" false)
    (:then
      (Variables|Default|SetHasLanded true)
      ;;[0 1 0.1364] This finishing on the client first can cause potential issues if the player swings the weapon before the server says its ok
      (CallFunction|DoTargetting)
      (CallFunction|OnPawnLanded))
    (:OnCompleted
      _)
    (:OnInterrupted
      _)
    (:OnCancelled
      _)
    (:OnNotifyBegin
      (if (== _notifytag (Variables|Const|GetRemoveAttackSuppressionTag))
        (Ability|EndAbility))))
  (Ability|Tasks|WaitNetSync "BothWait"
    (:OnSync
      (Ability|EndAbility))))

(event Ability|EventOnEndAbility
  (params (bWasCancelled :type Boolean))
  ;[1 1 1 1 64] Ability
  (Variables|Default|SetbCanPerformHit false)
  (if bWasCancelled
    ;;[1 0.006821 0] This ablity was cancelled from water, finish up.
    (CallFunction|UnbindFallbackLandCheck)
    (CallFunction|ForceLanding)
    (CallFunction|ForceLandedEvent))
  (|CallParentFunction bWasCancelled)
  (Utilities|Time|ClearandInvalidateTimerbyHandle (Variables|Runtime|GetPredictLandingTimerHandle))
  (GameplayTasks|EndTask (Variables|Runtime|GetWaitForLandAsyncTask))
  (CallFunction|UnbindAbilityPlayerPawnLanded))

(event Custom|ActivateLoopingWeaponFX
  (bind _returnvalue (Ability|GetAvatarActorfromActorInfo))
  ;[1 1 1] Cues
  (GameplayCue|AddGameplayCue _returnvalue (Variables|Const|GetGCNL_WeaponLoopingFX) (Class|AthenaFunctionLibrary|MakeSimpleFortEffectContext _returnvalue))
  (GameplayCue|AddGameplayCue _returnvalue (Variables|Const|GetGCL_JumpAttackFall)))

(event Custom|DeactivateLoopingWeaponFX
  (bind _returnvalue (Ability|GetAvatarActorfromActorInfo))
  ;[1 1 1] Cues
  (GameplayCue|RemoveGameplayCue _returnvalue (Variables|Const|GetGCNL_WeaponLoopingFX))
  (GameplayCue|RemoveGameplayCue _returnvalue (Variables|Const|GetGCL_JumpAttackFall)))

(event Custom|RemoveJumpAttackTag
  (Ability|RemoveGameplayEffectFromOwnerWithHandle (Variables|Default|GetJumpAttackAppliedHandle)))

(event Custom|OnPawnLanded
  ;[1 1 1] Cues
  (Utilities|FlowControl|DoOnce false
    (:Completed
      (if (not (Variables|Default|GetbIgnoreLanded))
        (Ability|ExecuteGameplayCueOnOwner (Variables|Const|GetGCN_JumpAttackSlam) (Class|AthenaFunctionLibrary|MakeSimpleFortEffectContext (Ability|GetAvatarActorfromActorInfo) (Ability|GetAvatarActorfromActorInfo)))))))

(event Custom|ResetOnPawnLanded
  ;[1 1 1] Cues
  (Utilities|FlowControl|DoOnce false
    (:Completed
      (if (not (Variables|Default|GetbIgnoreLanded))
        (Ability|ExecuteGameplayCueOnOwner (Variables|Const|GetGCN_JumpAttackSlam) (Class|AthenaFunctionLibrary|MakeSimpleFortEffectContext (Ability|GetAvatarActorfromActorInfo) (Ability|GetAvatarActorfromActorInfo)))))))

(event Custom|CancelIfClambering
  (bind _asynctaskproxy (Ability|Tasks|WaitGameplayTagAdd "(TagName=\"Clambering.State.LedgeMoving\")" (Ability|GetAvatarActorfromActorInfo))
    (:then
      (Variables|Default|SetClamberingStartTask _asynctaskproxy))
    (:Added
      (Variables|Default|SetbIgnoreLanded true)
      (Ability|CancelAbility))))

(event Custom|DoTargetting
  (bind _returnvalue (Ability|GetCurrentSourceWeapon))
  ;[1 1 1 1 64] Targeting
  (Utilities|FlowControl|DoOnce false
    (:Completed
      (if (Variables|Default|GetbCanPerformHit)
        (Development|PrintString "HIT!!!" true true "(R=0.937574,G=0.000000,B=1.000000,A=1.000000)" 5.0)
        (bind _outfortplayerpawn (CallFunction|GetAvatarasValidFortPlayerPawn)
          (:ValidFortPlayerPawn
            (bind _output_get (Variables|Default|SetFortPlayerpawn_0 _outfortplayerpawn))
            (bind (_asynctaskproxy _targetinghandle) (Class|AsyncActionPerformTargeting|PerformTargetingAsyncAction _output_get (Variables|Const|GetTargeting%20Preset))
              (:Targeted
                (for _array_element (TargetingSystem|TargetingResults|GetTargetingResults "/Script/TargetingSystem.Default__TargetingSubsystem" _targetinghandle)
                  (bind _output_get_2 (Variables|Default|SetTargetData (Ability|TargetData|AbilityTargetDatafromHitResult _array_element)))
                  (bind _output_get_3 (Variables|Default|SetHit%20Actor (.HitActor (Collision|BreakHitResult (Ability|EffectContext|GetHitResult (Ability|GetContextfromOwner _output_get_2))))))
                  (CallFunction|ApplyEffectstoTargets _output_get_2 true)
                  (Utilities|IsValid _output_get_3
                    (:Is%20Valid
                      (Utilities|IsValid _returnvalue
                        (:Is%20Valid
                          (FortWeapon|UseWeaponDurability(Scaled) _returnvalue (ScalableFloat|GetCachedScalableFloatValue (Variables|Default|GetHotfix_JumpAttackDurabilityCost))))))))))))))))

(event Custom|ResetDoTargetting
  (bind _returnvalue (Ability|GetCurrentSourceWeapon))
  ;[1 1 1 1 64] Targeting
  (Utilities|FlowControl|DoOnce false
    (:Completed
      (if (Variables|Default|GetbCanPerformHit)
        (Development|PrintString "HIT!!!" true true "(R=0.937574,G=0.000000,B=1.000000,A=1.000000)" 5.0)
        (bind _outfortplayerpawn (CallFunction|GetAvatarasValidFortPlayerPawn)
          (:ValidFortPlayerPawn
            (bind _output_get (Variables|Default|SetFortPlayerpawn_0 _outfortplayerpawn))
            (bind (_asynctaskproxy _targetinghandle) (Class|AsyncActionPerformTargeting|PerformTargetingAsyncAction _output_get (Variables|Const|GetTargeting%20Preset))
              (:Targeted
                (for _array_element (TargetingSystem|TargetingResults|GetTargetingResults "/Script/TargetingSystem.Default__TargetingSubsystem" _targetinghandle)
                  (bind _output_get_2 (Variables|Default|SetTargetData (Ability|TargetData|AbilityTargetDatafromHitResult _array_element)))
                  (bind _output_get_3 (Variables|Default|SetHit%20Actor (.HitActor (Collision|BreakHitResult (Ability|EffectContext|GetHitResult (Ability|GetContextfromOwner _output_get_2))))))
                  (CallFunction|ApplyEffectstoTargets _output_get_2 true)
                  (Utilities|IsValid _output_get_3
                    (:Is%20Valid
                      (Utilities|IsValid _returnvalue
                        (:Is%20Valid
                          (FortWeapon|UseWeaponDurability(Scaled) _returnvalue (ScalableFloat|GetCachedScalableFloatValue (Variables|Default|GetHotfix_JumpAttackDurabilityCost))))))))))))))))

(fn AbilityPlayerFallSpeed
  (bind _out_fort_player_pawn (Variables|Runtime|GetOut%20Fort%20Player%20Pawn))
  (bind (_returnvalue_x _returnvalue_y _returnvalue_z) (Math|Vector|BreakVector (Transformation|GetVelocity _out_fort_player_pawn)))
  (if (< _returnvalue_z 0)
    (return (Math|Float|Absolute(Float) _returnvalue_z))
    (else
      (return 0.0))))

(fn AbilityPlayerDistanceToVector
  (params (ImpactPoint :type Vector))
  (return (Math|Vector|VectorLength (- (Transformation|GetActorLocation (Variables|Runtime|GetOut%20Fort%20Player%20Pawn)) ImpactPoint))))

(fn DrawDebug
  (params (Center :type Vector))
  (if (Variables|Default|GetEnableDebugStuff)
    (Rendering|Debug|DrawDebugSphere Center 1e+02 12 "(R=1.000000,G=0.026704,B=0.000000,A=1.000000)" 2.0 5.0)))

(fn FloatWithinRange
  (params (Value :type Float%20%28double-precision%29) (Min :type Float%20%28double-precision%29) (Max :type Float%20%28double-precision%29))
  (return (and (>= Value Min) (<= Value Max))))

(fn GetNormalizedLaunchDirectionXY
  (bind (_returnvalue_x _returnvalue_y) (Math|Vector|BreakVector (Math|Vector|Normalize (- (Transformation|GetActorLocation (Variables|Default|GetHit%20Actor)) (Transformation|GetActorLocation (Ability|GetAvatarActorfromActorInfo))))))
  (return _returnvalue_x _returnvalue_y))

(fn GetApproximateFramesPerSecond
  (return (/ 1.0 (Utilities|Time|GetWorldDeltaSeconds))))

(fn GetMovementAmountPerFrame
  (return (* (Variables|/Script/Engine.CharacterMovementComponent|GetMaxWalkSpeed (Variables|/Script/Engine.Character|GetCharacterMovement (Variables|Runtime|GetOut%20Fort%20Player%20Pawn))) (Utilities|Time|GetWorldDeltaSeconds))))

(fn GetPlayerLocationWithCapsuleOffset
  (bind (_returnvalue_x _returnvalue_y _returnvalue_z) (Math|Vector|BreakVector (Transformation|GetActorLocation (Variables|Runtime|GetOut%20Fort%20Player%20Pawn))))
  (return _returnvalue_x _returnvalue_y (- _returnvalue_z (Components|Capsule|GetUnscaledCapsuleHalfHeight (Variables|/Script/Engine.Character|GetCapsuleComponent (Variables|Runtime|GetOut%20Fort%20Player%20Pawn))))))

(fn IsAbilityPlayerFalling
  (bind (_returnvalue_x _returnvalue_y _returnvalue_z) (Math|Vector|BreakVector (Transformation|GetVelocity (Variables|Runtime|GetOut%20Fort%20Player%20Pawn))))
  (return (< _returnvalue_z 0)))

(fn PredictionInRangeSeconds
  (return (- (Variables|Const|GetSecondsBeforeGroundToPlayPrediction) 0.1) (+ (Variables|Const|GetSecondsBeforeGroundToPlayPrediction) 0.1)))

(fn Reset
  (CallFunction|ResetPredictionHit)
  (CallFunction|ResetWaitforLand)
  (CallFunction|ResetPlayAblityPlayerPawnLandedEvent)
  (CallFunction|ResetActivation)
  (CallFunction|ResetDoTargetting)
  (CallFunction|ResetonPawnLanded))

(fn UpdateDelayTimeFromGravity
  (bind _outfortplayerpawn (CallFunction|GetAvatarasValidFortPlayerPawn)
    (:ValidFortPlayerPawn
      (Variables|Default|SetForce%20End%20Delay (+ (Variables|Default|GetBaseForceEndDuration) (/ (Variables|Default|GetDefaultGravityValue) (Components|Movement|GetGravityZ (Pawn|GetMovementComponent _outfortplayerpawn))))))))
