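; Chainsaw primary melee gameplay ability that turns an attack input into a jump attack when the player is airborne.
; On activation or retry after failure, it clears balloons and, if falling or jetpacking without PreventJump, sends a jump-and-wait-for-landing event.
; Activation also stops the saw scooter and ends immediately, with no combo logic in this graph.
; A failure-reason override silences the message when chainsaw primary attack is blocked.
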(event Custom|Retry%20After%20Fail
  (bind _outfortplayerpawn (CallFunction|GetAvatarasValidFortPlayerPawn)
    (:ValidFortPlayerPawn
      (Ability|SendGameplayEventtoActor _outfortplayerpawn "(TagName=\"Event.Athena.Balloons.RemoveAll\")")
      (if (and (or (AI|Components|NavMovement|IsFalling (Variables|/Script/Engine.Character|GetCharacterMovement _outfortplayerpawn)) (GameplayTags|HasMatchingGameplayTag _outfortplayerpawn "(TagName=\"Gameplay.Action.Player.Jetpack\")")) (not (GameplayTags|HasMatchingGameplayTag _outfortplayerpawn "(TagName=\"Gameplay.Action.Player.PreventJump\")")))
        ;[1 1 1] Trigger Jump Attack
        (Ability|SendGameplayEventtoActor _outfortplayerpawn (Variables|Tags|GetEventJumpAndWaitForLanding) (Utilities|Struct|MakeGameplayEventData (Variables|Tags|GetEventJumpAndWaitForLanding) (Ability|GetAvatarActorfromActorInfo) 0 0 0 "()" "(GameplayTags=)" "(GameplayTags=)" (Variables|Const|GetJumpMagnitude)))
        (Ability|EndAbility)))
    (:AvatarCastFailed
      (Ability|EndAbility))))

(event Ability|EventActivateAbility
  (Development|PrintString "ACTIVATE" true true "(R=1.000000,G=0.000000,B=0.954631,A=1.000000)")
  (Ability|SendGameplayEventtoActor (CallFunction|GetAvatarasFortPlayerPawn) "(TagName=\"Event.Athena.Chainsaw.StopSawScooter.FromAttack\")")
  (bind _asynctaskproxy (Ability|Tasks|WaitInputRelease false)
    (:then
      (Variables|Task|SetWaitInputReleasedAsyncTask _asynctaskproxy)
      (bind _outfortplayerpawn (CallFunction|GetAvatarasValidFortPlayerPawn)
        (:ValidFortPlayerPawn
          (Ability|SendGameplayEventtoActor _outfortplayerpawn "(TagName=\"Event.Athena.Balloons.RemoveAll\")")
          (if (and (or (AI|Components|NavMovement|IsFalling (Variables|/Script/Engine.Character|GetCharacterMovement _outfortplayerpawn)) (GameplayTags|HasMatchingGameplayTag _outfortplayerpawn "(TagName=\"Gameplay.Action.Player.Jetpack\")")) (not (GameplayTags|HasMatchingGameplayTag _outfortplayerpawn "(TagName=\"Gameplay.Action.Player.PreventJump\")")))
            ;[1 1 1] Trigger Jump Attack
            (Ability|SendGameplayEventtoActor _outfortplayerpawn (Variables|Tags|GetEventJumpAndWaitForLanding) (Utilities|Struct|MakeGameplayEventData (Variables|Tags|GetEventJumpAndWaitForLanding) (Ability|GetAvatarActorfromActorInfo) 0 0 0 "()" "(GameplayTags=)" "(GameplayTags=)" (Variables|Const|GetJumpMagnitude)))
            (Ability|EndAbility)))
        (:AvatarCastFailed
          _)))
    (:OnRelease
      _))
  (Ability|EndAbility))

(event Ability|EventOnEndAbility
  (params (bWasCancelled :type Boolean))
  (GameplayTasks|EndTask (Variables|Task|GetWaitInputReleasedAsyncTask)))

(fn OverrideFailedReason
  (params (FailedReason :type Gameplay%20Tag%20Container%20Structure%20%28by%20ref%29))
  (if (GameplayTags|HasAllTags FailedReason (GameplayTags|MakeLiteralGameplayTagContainer "(GameplayTags=((TagName=\"Activate.Failed.Tags.Blocked\"),(TagName=\"Granted.Athena.Chainsaw.BlockPrimaryAttack\")))") true)
    (return "(GameplayTags=((TagName=\"Athena.Ability.Failed.NoMessage\")))")
    (else
      (return FailedReason))))
