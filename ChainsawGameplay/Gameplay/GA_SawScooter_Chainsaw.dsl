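; Gameplay ability that runs the chainsaw "saw scooter" ride, ending on jump, overheat, attack, ledge exit or button release after a minimum time.
; Activation starts a custom movement mode, camera modifier and steeper walkable angle.
; Timers align mesh and wheel FX to the ground, damage buildings and vehicles, and auto-attack enemies via a blocking tag that triggers the stop event.
; Ending restores movement, grants fall immunity and launches the player upward unless exiting from a ledge.
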(event Custom|ReachedMinTime
  ;[0.6353 1 0.5625] On Activate Ability
  (if (not (GameplayTags|HasMatchingGameplayTag (Variables|Default|GetPlayer) "(TagName=\"Granted.Athena.Chainsaw.HoldingScooterButton\")"))
    (Ability|HasAuthority
      (:True
        (Ability|EndAbility)))))

(event Custom|OnNormalJump
  ;[0.6353 1 0.5625] On Activate Ability
  (Ability|EndAbility))

(event Custom|OnUnequip
  ;[0.6353 1 0.5625] On Activate Ability
  ;;[1 0 0] TODO re-hook this up with new weapon
  (Ability|EndAbility))

(event Ability|EventActivateAbility
  (bind _minscooteringtime (Variables|Default|GetMinScooteringTime))
  (bind _returnvalue (ScalableFloat|GetValueatLevel _minscooteringtime))
  (bind _returnvalue_2 (Utilities|Time|GetGameTimeinSeconds))
  (bind _player (Variables|Default|GetPlayer))
  (bind _returnvalue_3 (Fortnite|GetFortPlayerControllerfromActor _player))
  (bind (_bsuccessfullyfoundattribute _returnvalue_4) (Ability|Attribute|GetFloatAttribute (Variables|Default|GetPlayer) "(AttributeName=\"Overheat\",Attribute=/Script/ChainsawGameplayRuntime.ChainsawAttributeSet:Overheat,AttributeOwner=\"/Script/CoreUObject.Class'/Script/ChainsawGameplayRuntime.ChainsawAttributeSet'\")"))
  (bind _abilityactivatedtime (Variables|Default|GetAbilityActivatedTime))
  (bind _returnvalue_5 (- _returnvalue_2 _abilityactivatedtime))
  ;[0.6353 1 0.5625] On Activate Ability
  (|CallParentFunction)
  (Utilities|FlowControl|Sequence
    (:then_0
      (Variables|Default|SetAttackedPawn 0)
      (bind _asplayer_pawn_athena (Utilities|Casting|CastToPlayerPawn_Athena (Ability|GetActivatingPawn)))
      (bind _output_get (Variables|Default|SetPlayer _asplayer_pawn_athena))
      (bind _asb_weapon_chainsaw (Utilities|Casting|CastToB_Weapon_Chainsaw (Variables|/Script/FortniteGame.FortPawn|GetCurrentWeapon _output_get)))
      (Variables|Default|SetWeapon _asb_weapon_chainsaw)
      (Variables|Default|SetAbilityActivatedTime (Utilities|Time|GetGameTimeinSeconds)))
    (:then_1
      ;;[1 0 0] TODO re-hook this up with new weapon
      (Pawn|BindEventtoOnPawnJumped (Variables|Default|GetPlayer) (delegate Custom|OnNormalJump))
      (bind (_asynctaskproxy _payload) (Ability|Tasks|WaitGameplayEvent "(TagName=\"Event.Athena.Chainsaw.StopSawScooter\")" 0 false false)
        (:EventReceived
          (Variables|Default|SetExitedFromLedge (== (Utilities|Struct|BreakGameplayEventData _payload) "(TagName=\"Event.Athena.Chainsaw.StopSawScooter.FromLedge\")"))
          (Utilities|FlowControl|Sequence
            (:then_0
              (if (> _returnvalue_5 1.5)
                (Variables|/ChainsawGameplay/Gameplay/B_Weapon_Chainsaw.B_Weapon_Chainsaw_C|SetHasHeldScooterButtonBefore true (Variables|Default|GetWeapon))))
            (:then_1
              (if (Math|Boolean|ORBoolean (Variables|Default|GetExitedFromLedge) (== (Utilities|Struct|BreakGameplayEventData _payload) "(TagName=\"Event.Athena.Chainsaw.StopSawScooter.FromAttack\")") (> _returnvalue_5 _returnvalue))
                (Ability|HasAuthority
                  (:True
                    (Ability|EndAbility)))
                (else
                  (Utilities|FlowControl|Sequence
                    (:then_0
                      (Utilities|Time|SetTimerbyEvent (delegate Custom|ReachedMinTime) (- _returnvalue _returnvalue_5)))
                    (:then_1
                      (if (< _returnvalue_5 0.5)
                        (if (not (Variables|/ChainsawGameplay/Gameplay/B_Weapon_Chainsaw.B_Weapon_Chainsaw_C|GetHasHeldScooterButtonBefore (Variables|Default|GetWeapon)))
                          (Ability|SendGameplayEventtoActor (Variables|Default|GetPlayer) "(TagName=\"Event.Athena.Chainsaw.DidntHoldScooterButton\")"))))))))))))
    (:then_2
      (Ability|Async|EndAction (Variables|Default|GetAction_WaitForTagAdded))
      (bind _asynctaskproxy_2 (Ability|Tasks|WaitGameplayTagAddToActor (Variables|Default|GetPlayer) "(TagName=\"Granted.Athena.Chainsaw.BlockScootering.FromScooterAttack\")")
        (:then
          (Variables|Default|SetAction_WaitForTagAdded _asynctaskproxy_2)
          (Variables|Default|SetStartedTime (Utilities|Time|GetGameTimeinSeconds))
          (Utilities|IsValid _returnvalue_3
            (:Is%20Valid
              (bind _returnvalue_6 (Game|Player|AddNewCameraModifier (Variables|/Script/Engine.PlayerController|GetPlayerCameraManager _returnvalue_3) "/ChainsawGameplay/Gameplay/CameraModifier_SawScooter_Chainsaw.CameraModifier_SawScooter_Chainsaw_C"))
              (Variables|Default|SetCameraMod_ _returnvalue_6))))
        (:Added
          (CallFunction|SendScooterAttackEvent))))
    (:then_4
      (Variables|Default|SetNormalWalkableFloorAngle 44.76)
      (Pawn|Components|CharacterMovement|SetWalkableFloorAngle (Variables|/Script/Engine.Character|GetCharacterMovement (Variables|Default|GetPlayer)) (ScalableFloat|GetValueatLevel (Variables|Default|GetSF_NewWalkableAngle)))
      (bind _returnvalue_7 (Class|FortMovementModeExtBlueprintLibrary|ActivateRegisteredMovementModeExtension "/ChainsawGameplay/Gameplay/MME/MME_Logic_Chainsaw.MME_Logic_Chainsaw_C" (Variables|Default|GetPlayer)))
      (bind _output_get_2 (Variables|Default|SetMME_Ref _returnvalue_7))
      (bind _asmme_runtime_chainsaw (Utilities|Casting|CastToMME_Runtime_Chainsaw _output_get_2))
      (Variables|Default|SetChainsawRuntime _asmme_runtime_chainsaw))
    (:then_5
      (CallFunction|StartUpdating)
      (CallFunction|ResetAttackTrigger)
      (Utilities|Time|SetTimerbyEvent (delegate Custom|CheckIfOutOfFuel) 0.1 true true)
      (if (<= _returnvalue_4 0)
        (Ability|ApplyGameplayEffectToOwner "/ChainsawGameplay/Gameplay/GameplayEffects/GE_SetFuelToMax_Chainsaw.GE_SetFuelToMax_Chainsaw_C")))))

(event Custom|CheckIfOutOfFuel
  (bind (_bsuccessfullyfoundattribute _returnvalue) (Ability|Attribute|GetFloatAttribute (Variables|Default|GetPlayer) "(AttributeName=\"Overheat\",Attribute=/Script/ChainsawGameplayRuntime.ChainsawAttributeSet:Overheat,AttributeOwner=\"/Script/CoreUObject.Class'/Script/ChainsawGameplayRuntime.ChainsawAttributeSet'\")"))
  ;[0.6353 1 0.5625] On Activate Ability
  (if (>= _returnvalue 1e+02)
    (GameplayCue|ExecuteGameplayCue%28Burst%29Local (Variables|Default|GetPlayer) (Variables|Default|GetGC_Overheat) (Ability|GameplayCue|MakeGameplayCueParameters 0.0))
    (Ability|EndAbility)))

(event Custom|SendScooterAttackEvent
  (Utilities|FlowControl|Sequence
    (:then_0
      (if (GameplayTags|HasMatchingGameplayTag (Variables|Default|GetPlayer) "(TagName=\"Granted.Athena.Chainsaw.BlockScootering.FromScooterAttack\")")
        (Ability|SendGameplayEventtoActor (Variables|Default|GetPlayer) "(TagName=\"Event.Athena.Chainsaw.StopSawScooter.FromAttack\")" (Utilities|Struct|MakeGameplayEventData "(TagName=\"Event.Athena.Chainsaw.StopSawScooter.FromAttack\")" 0 (Variables|Default|GetAttackedPawn)))))
    (:then_1
      (Variables|Default|SetAttackedPawn 0))))

(event Ability|EventOnEndAbility
  (params (bWasCancelled :type Boolean))
  (bind _player (Variables|Default|GetPlayer))
  ;[1 0.5417 0.5417] On End Ability
  (Utilities|FlowControl|Sequence
    (:then_0
      (Class|CameraModifierSawScooterChainsaw|WindDownandRemove (Variables|Default|GetCameraMod_))
      (Ability|ApplyGameplayEffectToOwner "/Game/Athena/Items/Parents/FallImmunity/GE_Athena_TillLandFallDamageImmunity_NoFX_Grant_Parent.GE_Athena_TillLandFallDamageImmunity_NoFX_Grant_Parent_C")
      (Ability|ApplyGameplayEffectToOwner "/ChainsawGameplay/Gameplay/GameplayEffects/GE_BlockRefuelingInitially_Chainsaw.GE_BlockRefuelingInitially_Chainsaw_C")
      (Pawn|Components|CharacterMovement|SetWalkableFloorAngle (Variables|/Script/Engine.Character|GetCharacterMovement (Variables|Default|GetPlayer)) (Variables|Default|GetNormalWalkableFloorAngle))
      (Development|PrintString "End Ability")
      (Ability|RemoveGameplayEffectFromOwnerWithHandle (Variables|Default|GetGE_WheelMovement))
      (Transformation|SetRelativeRotation (Variables|/Script/Engine.Character|GetMesh (Variables|Default|GetPlayer)) "0.000000,-90.000000,0.000000" false false)
      (bind _asfort_movement_comp_character (Utilities|Casting|CastToFortMovementComp_Character (Variables|/Script/Engine.Character|GetCharacterMovement _player)))
      (Class|FortMovementCompCharacter|CancelActiveExtensionwithTag _asfort_movement_comp_character (GameplayTags|MakeGameplayTagContainerfromTag "(TagName=\"MovementMode.Custom.Chainsaw\")"))
      (if (not (Variables|Default|GetExitedFromLedge))
        (Character|LaunchCharacter _player "0.000000,0.000000,750.000000" false true)))))

(event Custom|Check%20Water%20Interaction
  (params (Surface :type Integer) (Actor :type Actor%20Object%20Reference))
  (if (== Surface 8)
    (Utilities|FlowControl|Sequence
      (:then_0
        (if (not (== (Variables|Default|GetWaterBodyActor) Actor))
          (bind _output_get (Variables|Default|SetWaterBodyActor Actor))
          (bind _aswater_body (Utilities|Casting|CastToWaterBody _output_get))
          (Water|SetWaterBodyComponent (Variables|Default|GetWheel%20FX) "WaterBodyActor" (Water|GetWaterBodyComponent _aswater_body))))
      (:then_1
        (Niagara|SetNiagaraVariable(Bool) (Variables|Default|GetWheel%20FX) "IsInWater" true)))
    (else
      (Niagara|SetNiagaraVariable(Bool) (Variables|Default|GetWheel%20FX) "IsInWater" false))))

(event Custom|SlowUpdate
  (bind _wheel_fx (Variables|Default|GetWheel%20FX))
  ;[1 1 1] Slow Update
  (bind (_outhit _returnvalue) (Collision|LineTraceForObjects (Transformation|GetActorLocation (Variables|Default|GetPlayer)) (- (Transformation|GetActorLocation (Variables|Default|GetPlayer)) "0.000000,0.000000,160.000000") (Utilities|Array|MakeArray "ObjectTypeQuery1" "ObjectTypeQuery13") false (Utilities|Array|MakeArray (Variables|Default|GetPlayer))))
  (if _returnvalue
    (Variables|Default|SetGoalUpVec (.Normal (Collision|BreakHitResult _outhit)))
    (Variables|Default|SetRightVec (CallFunction|GetMovementDirectionSafe))
    (Niagara|SetNiagaraVariable(Int32) (Variables|Default|GetWheel%20FX) "PhysicalSurface" (Math|Conversions|ToInteger(Byte) (Variables|/Script/PhysicsCore.PhysicalMaterial|GetSurfaceType (.PhysMat (Collision|BreakHitResult _outhit)))))
    (CallFunction|CheckWaterInteraction (Math|Conversions|ToInteger(Byte) (Variables|/Script/PhysicsCore.PhysicalMaterial|GetSurfaceType (.PhysMat (Collision|BreakHitResult _outhit)))) (.HitActor (Collision|BreakHitResult _outhit)))
    (Utilities|IsServer
      (:false
        (bind _updated_variable (Utilities|FlowControl|DeltaGateWildcard (Variables|Default|GetPrevSurface) (Variables|/Script/PhysicsCore.PhysicalMaterial|GetSurfaceType (.PhysMat (Collision|BreakHitResult _outhit))))
          (:Success
            (bind _output_get (Variables|Default|SetPrevSurface _updated_variable))
            (bind _weapon (Variables|Default|GetWeapon)
              (:then
                (Class|BWeaponChainsaw|SetUpCurrentSurface _weapon _output_get))
              (:else))))))
    (else
      (Niagara|SetNiagaraVariable(Int32) _wheel_fx "PhysicalSurface" 0)
      (Niagara|SetNiagaraVariable(Bool) _wheel_fx "IsInWater" false)
      (Variables|Default|SetGoalUpVec (Math|Vector|MakeVector 0.0 0.0 1.0))
      (Variables|Default|SetRightVec (CallFunction|GetMovementDirectionSafe)))))

(event Custom|End%20Water%20Interaction)

(event Custom|StartUpdating
  ;[1 1 1] Start Updating
  (CallFunction|SlowUpdate)
  (Utilities|Time|SetTimerbyEvent (EventDispatchers|CreateEvent :!event Custom|SlowUpdate) 0.15 true)
  (Variables|Default|SetPlayerCurrentRotation (Transformation|GetWorldRotation (Variables|/Script/Engine.Character|GetMesh (Variables|Default|GetPlayer))))
  (Utilities|Time|SetTimerbyEvent (EventDispatchers|CreateEvent :!event Custom|Update) 0.0001 true true)
  (bind _output_get (Variables|Default|SetCurrentUpVec "0.000000,0.000000,1.000000"))
  (Variables|Default|SetGoalUpVec _output_get)
  (Variables|Default|SetRightVec (Transformation|GetActorForwardVector (Variables|Default|GetPlayer)))
  (Variables|Default|SetPlayerGoalRotation (Math|Rotator|MakeRotfromZY (Variables|Default|GetCurrentUpVec) (Variables|Default|GetRightVec)))
  (bind _returnvalue (Utilities|Time|SetTimerbyEvent (delegate Custom|DealEnvironmentDamage) (ScalableFloat|GetValueatLevel (Variables|Default|GetSF_TracesUpdateFrequency)) true true))
  (Variables|Default|SetTraceTimer _returnvalue))

(event Custom|DealEnvironmentDamage
  (bind _player (Variables|Default|GetPlayer))
  (bind _returnvalue (Transformation|GetActorLocation _player))
  (Utilities|FlowControl|Sequence
    (:then_0
      (bind _outhits (Collision|MultiSphereTraceByProfile _returnvalue (+ _returnvalue (* (CallFunction|GetMovementDirectionSafe) (ScalableFloat|GetValueatLevel (Variables|Default|GetSF_EnvironmentDamageTrace_Distance)))) (ScalableFloat|GetValueatLevel (Variables|Default|GetSF_EnvironmentDamageTrace_Radius)) "OverlapAll" false (Utilities|Array|MakeArray (Variables|Default|GetPlayer)) "None" true "(R=1.000000,G=0.000000,B=0.000000,A=1.000000)" "(R=0.000000,G=1.000000,B=0.000000,A=1.000000)" 0.2))
      (Variables|Default|SetHitAnyBuildingActors false)
      (for _array_element _outhits
        (bind _asbuilding_actor (Utilities|Casting|CastToBuildingActor (.HitActor (Collision|BreakHitResult _array_element)))
          (:then
            (if (Variables|/Script/Engine.Actor|GetbCanBeDamaged _asbuilding_actor)
              (GameplayEffects|ApplyGameplayEffectToTarget :self (Ability|GetFortAbilitySystemComponent (Variables|Default|GetPlayer)) :GameplayEffectClass "/ChainsawGameplay/Gameplay/GameplayEffects/GE_Damage_SawScooter_Chainsaw.GE_Damage_SawScooter_Chainsaw_C" :Target (Ability|GetFortAbilitySystemComponent (.HitActor (Collision|BreakHitResult _array_element))) :Level 0.0)
              (Variables|Default|SetHitAnyBuildingActors true)))
          (:CastFailed
            (Utilities|Casting|CastToFortVehicleInterface (.HitActor (Collision|BreakHitResult _array_element)))
            (GameplayEffects|ApplyGameplayEffectToTarget :self (Ability|GetFortAbilitySystemComponent (Variables|Default|GetPlayer)) :GameplayEffectClass "/ChainsawGameplay/Gameplay/GameplayEffects/GE_VehicleDamage_SawScooter_Chainsaw.GE_VehicleDamage_SawScooter_Chainsaw_C" :Target (Ability|GetFortAbilitySystemComponent (.HitActor (Collision|BreakHitResult _array_element))) :Level 0.0)
            (Audio|BroadcastAbilityImpactatLocation 0 (.HitActor (Collision|BreakHitResult _array_element)) (* (CallFunction|GetKnockbackDirection 0.7 0.5) 1e+03) 7.8e+02) ; This is the secret node that launches vehicles. Note it needs the 780 radius to work.
            (GameplayEffects|ApplyGameplayEffectToTarget :self (Ability|GetFortAbilitySystemComponent (Variables|Default|GetPlayer)) :GameplayEffectClass "/ChainsawGameplay/Gameplay/GameplayEffects/GE_TemporaryScooterDamageImmunity_Chainsaw.GE_TemporaryScooterDamageImmunity_Chainsaw_C" :Target (Ability|GetFortAbilitySystemComponent (.HitActor (Collision|BreakHitResult _array_element))) :Level 0.0))))
      (if (Variables|Default|GetHitAnyBuildingActors)
        (Pawn|Components|CharacterMovement|SetWalkableFloorAngle (Variables|/Script/Engine.Character|GetCharacterMovement (Variables|Default|GetPlayer)) (Variables|Default|GetNormalWalkableFloorAngle))
        (Ability|ApplyGameplayEffectToOwner "/ChainsawGameplay/Gameplay/GameplayEffects/GE_SawScooter_BlockedByObstacle.GE_SawScooter_BlockedByObstacle_C")
        (else
          (Pawn|Components|CharacterMovement|SetWalkableFloorAngle (Variables|/Script/Engine.Character|GetCharacterMovement (Variables|Default|GetPlayer)) (ScalableFloat|GetValueatLevel (Variables|Default|GetSF_NewWalkableAngle))))))
    (:then_2
      (if (ScalableFloat|GetValueasBool (Variables|Default|GetSF_AutoAttack_Enabled))
        (bind _outhits_2 (Collision|MultiSphereTraceByProfile (+ _returnvalue (* (Math|Vector|GetForwardVector (Pawn|GetDesiredRotation (Pawn|GetController (Variables|Default|GetPlayer)))) (* (Variables|/Script/Engine.CapsuleComponent|GetCapsuleRadius (Variables|/Script/Engine.Character|GetCapsuleComponent (Variables|Default|GetPlayer))) (ScalableFloat|GetValueatLevel (Variables|Default|GetSF_AutoAttack_Trace_SourceForwardOffset))))) (+ _returnvalue (* (Math|Vector|GetForwardVector (Pawn|GetDesiredRotation (Pawn|GetController (Variables|Default|GetPlayer)))) (ScalableFloat|GetValueatLevel (Variables|Default|GetSF_AutoAttack_Trace_Distance)))) (* (Variables|/Script/Engine.CapsuleComponent|GetCapsuleRadius (Variables|/Script/Engine.Character|GetCapsuleComponent (Variables|Default|GetPlayer))) (ScalableFloat|GetValueatLevel (Variables|Default|GetSF_AutoAttack_Trace_Radius))) "FortTriggerOnlyPawns" false (Utilities|Array|MakeArray (Variables|Default|GetPlayer)) "None" true "(R=1.000000,G=0.000000,B=0.000000,A=1.000000)" "(R=0.000000,G=1.000000,B=0.000000,A=1.000000)" 0.05)) ; Larger trace for players
        (for _array_element_2 _outhits_2
          (bind _aspawn (Utilities|Casting|CastToPawn (.HitActor (Collision|BreakHitResult _array_element_2))))
          (if (not (Team|OnSameTeam _aspawn (Variables|Default|GetPlayer)))
            (Utilities|FlowControl|DoOnce false
              (:Completed
                (Variables|Default|SetAttackedPawn _aspawn)
                (GameplayEffects|ApplyGameplayEffectToTarget :self (Ability|GetFortAbilitySystemComponent (Variables|Default|GetPlayer)) :GameplayEffectClass "/ChainsawGameplay/Gameplay/GameplayEffects/GE_TemporarilyBlockSawScooter_FromAttack_Chainsaw.GE_TemporarilyBlockSawScooter_FromAttack_Chainsaw_C" :Target (Ability|GetFortAbilitySystemComponent (Variables|Default|GetPlayer)) :Level 0.0)
                (Utilities|Time|ClearandInvalidateTimerbyHandle (Variables|Default|GetTraceTimer))))))))))

(event Custom|Update
  ;[1 1 1] Update
  (Utilities|FlowControl|Sequence
    (:then_0
      (Variables|Default|SetCurrentUpVec (Math|Vector|GetForwardVector (Math|Interpolation|RInterpTo (Math|Conversions|RotationFromXVector (Variables|Default|GetCurrentUpVec)) (Math|Conversions|RotationFromXVector (Variables|Default|GetGoalUpVec)) (Utilities|Time|GetWorldDeltaSeconds) (Math|Float|SelectFloat 0.7 (Math|Float|SelectFloat 1e+01 5.0 (Variables|/Script/FortniteGame.FortPlayerPawn|GetbIsTacticalSprinting (Variables|Default|GetPlayer))) (AI|Components|NavMovement|IsFalling (Variables|/Script/Engine.Character|GetCharacterMovement (Variables|Default|GetPlayer)))))))
      (if (not (Variables|Default|GetHitAnyBuildingActors))
        (Variables|Default|SetPlayerGoalRotation (Math|Rotator|MakeRotfromZY (Variables|Default|GetCurrentUpVec) (Variables|Default|GetRightVec))))
      (Variables|Default|SetPlayerCurrentRotation (Math|Interpolation|RInterpTo (Variables|Default|GetPlayerCurrentRotation) (Variables|Default|GetPlayerGoalRotation) (Utilities|Time|GetWorldDeltaSeconds) 1e+01))
      (Variables|Default|SetCurrentTilt (Math|Interpolation|FInterpTo (Variables|Default|GetCurrentTilt) (CallFunction|GetGoalTilt) (Utilities|Time|GetWorldDeltaSeconds) 4.5))
      (Niagara|SetNiagaraVariable(Float) (Variables|Default|GetWheel%20FX) "WheelTiltAngle" (Variables|Default|GetCurrentTilt))
      (Transformation|SetWorldRotation (Variables|/Script/Engine.Character|GetMesh (Variables|Default|GetPlayer)) (Variables|Default|GetPlayerCurrentRotation) false false))))

(event Custom|ResetAttackTrigger
  (Utilities|FlowControl|DoOnce false
    (:Completed
      (Variables|Default|SetAttackedPawn (stale Custom|DealEnvironmentDamage _aspawn))
      (GameplayEffects|ApplyGameplayEffectToTarget :self (Ability|GetFortAbilitySystemComponent (Variables|Default|GetPlayer)) :GameplayEffectClass "/ChainsawGameplay/Gameplay/GameplayEffects/GE_TemporarilyBlockSawScooter_FromAttack_Chainsaw.GE_TemporarilyBlockSawScooter_FromAttack_Chainsaw_C" :Target (Ability|GetFortAbilitySystemComponent (Variables|Default|GetPlayer)) :Level 0.0)
      (Utilities|Time|ClearandInvalidateTimerbyHandle (Variables|Default|GetTraceTimer)))))

(fn GetGoalTilt
  (return (* (Utilities|RemapFloat (Math|Vector|DotProduct (/ (Pawn|Components|CharacterMovement|GetCurrentAcceleration (Variables|/Script/Engine.Character|GetCharacterMovement (Variables|Default|GetPlayer))) (Pawn|Components|CharacterMovement|GetMaxAcceleration (Variables|/Script/Engine.Character|GetCharacterMovement (Variables|Default|GetPlayer)))) (Math|Vector|CrossProduct (Math|Vector|Normalize (Transformation|GetVelocity (Variables|Default|GetPlayer))) "0.000000,0.000000,1.000000")) "(X=-0.600000,Y=0.600000)" "(X=-30.000000,Y=30.000000)" true) (Math|Float|SelectFloat 0.9 1.1 (> (Utilities|RemapFloat (Math|Vector|DotProduct (/ (Pawn|Components|CharacterMovement|GetCurrentAcceleration (Variables|/Script/Engine.Character|GetCharacterMovement (Variables|Default|GetPlayer))) (Pawn|Components|CharacterMovement|GetMaxAcceleration (Variables|/Script/Engine.Character|GetCharacterMovement (Variables|Default|GetPlayer)))) (Math|Vector|CrossProduct (Math|Vector|Normalize (Transformation|GetVelocity (Variables|Default|GetPlayer))) "0.000000,0.000000,1.000000")) "(X=-0.600000,Y=0.600000)" "(X=-30.000000,Y=30.000000)" true) 0)))))

(fn GetKnockbackDirection
  (params (UpAmount :type Float%20%28double-precision%29) (RightAmount :type Float%20%28single-precision%29))
  (return (Utilities|Operators|Add (Math|Vector|Normalize2D(Vector) (Math|Vector|GetForwardVector (Pawn|GetBaseAimRotation (Variables|Default|GetPlayer)))) (Math|Vector|MakeVector 0.0 0.0 UpAmount) (* (Math|Vector|CrossProduct (Math|Vector|Normalize2D(Vector) (Math|Vector|GetForwardVector (Pawn|GetBaseAimRotation (Variables|Default|GetPlayer)))) "0.000000,0.000000,-1.000000") RightAmount))))

(fn GetMovementDirectionSafe
  (return (Math|Vector|SelectVector (Math|Vector|Normalize2D(Vector) (Transformation|GetVelocity (Variables|Default|GetPlayer))) (Variables|Default|GetRightVec) (> (Math|Vector|VectorLengthSquared (Transformation|GetVelocity (Variables|Default|GetPlayer))) 1.0))))

(fn OverrideFailedReason
  (params (FailedReason :type Gameplay%20Tag%20Container%20Structure%20%28by%20ref%29))
  (bind _returnvalue (GameplayTags|HasMatchingGameplayTag(Message) (Ability|GetActivatingPawn) "(TagName=\"Granted.Athena.HumbleMop.Wheel.Rolling\")"))
  (if _returnvalue
    (return "(GameplayTags=((TagName=\"Athena.Ability.Failed.Cooldown\")))")
    (else
      (return FailedReason))))

(fn OnRep_AttackedPawn)

(fn ShouldPlayFailedMontage
  (params (FailedReason :type Gameplay%20Tag%20Container%20Structure%20%28by%20ref%29))
  (return true))
