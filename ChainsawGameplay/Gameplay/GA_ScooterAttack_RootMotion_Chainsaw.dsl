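; Chainsaw gameplay ability for the lunging attack out of the saw scooter, gated by a hotfix flag.
; Gameplay events start the attack and can supply a preset target.
; Root motion moves to that target or an aim-assist pawn, otherwise slides forward, optionally floor-aligned.
; Montage notifies apply damage and hit cues via line-of-sight-checked preset target or Targeting System query; a finish notify or failsafe timer ends it.
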(event Ability|EventActivateAbilityFromEvent
  (params (EventData :type Gameplay%20Event%20Data%20Structure%20%28by%20ref%29))
  (bind (_eventtag _instigator _target) (Utilities|Struct|BreakGameplayEventData EventData))
  ;[1 1 1] Activate
  (|CallParentFunction EventData)
  (if (ScalableFloat|GetValueasBool (Variables|Default|GetScooterAttackEnabledHotfixRow))
    (bind _outfortplayerpawn (CallFunction|GetAvatarasValidFortPlayerPawn)
      (:ValidFortPlayerPawn
        (bind _output_get (Variables|Default|SetFort%20Player%20Pawn _outfortplayerpawn))
        (bind _asb_weapon_chainsaw (Utilities|Casting|CastToB_Weapon_Chainsaw (Weapon|GetCurrentWeapon _output_get)))
        (Variables|Default|SetWeapon _asb_weapon_chainsaw)
        (if (== _eventtag "(TagName=\"Event.Athena.Chainsaw.DoScooterAttack\")")
          (Utilities|IsValid _target
            (:Is%20Valid
              (Variables|Default|SetPresetTarget _target)
              (Variables|Default|SetHasPresetTarget true))
            (:Is%20Not%20Valid
              _))
          (CallFunction|OnSawScooterEnded)
          (else
            (Variables|Default|SetHasPresetTarget false)
            (Ability|Tasks|WaitGameplayEvent "(TagName=\"Event.Athena.Chainsaw.StopSawScooter\")" (Variables|Default|GetFort%20Player%20Pawn) true
              (:then
                (bind (_asynctaskproxy _payload) (Ability|Tasks|WaitGameplayEvent "(TagName=\"Event.Athena.Chainsaw.StopSawScooter.FromAttack\")" (Variables|Default|GetFort%20Player%20Pawn) true)
                  (:then
                    (Ability|Tasks|WaitGameplayEvent "(TagName=\"Event.Athena.Chainsaw.DoScooterAttack\")" 0 true
                      (:EventReceived
                        (Ability|EndAbility))))
                  (:EventReceived
                    (Utilities|FlowControl|Sequence
                      (:then_0
                        (Utilities|IsValid (.Target (Utilities|Struct|BreakGameplayEventData _payload))
                          (:Is%20Valid
                            (Variables|Default|SetPresetTarget (.Target (Utilities|Struct|BreakGameplayEventData _payload)))
                            (Variables|Default|SetHasPresetTarget true))))
                      (:then_1
                        (CallFunction|OnSawScooterEnded))))))
              (:EventReceived
                (if (not (Variables|Default|GetDoingScooterAttack))
                  (Ability|EndAbility)))))))
      (:AvatarCastFailed
        (Ability|EndAbility)))
    (else
      (Ability|EndAbility))))

(event Custom|OnSawScooterEnded
  (bind (_montagetoplay _animplayrate _animrootmotiontranslationscale) (Utilities|Struct|BreakFortGameplayAbilityMontageInfo (Variables|Animation|GetMontageInfo)))
  (bind _returnvalue (Ability|GetCurrentSourceWeapon))
  (bind _sf_forceendfailsafetime (Variables|Default|GetSF_ForceEndFailsafeTime))
  (bind _returnvalue_2 (ScalableFloat|GetValueatLevel _sf_forceendfailsafetime))
  ;[1 1 1] Do Attack
  (Utilities|FlowControl|DoOnce false
    (:Completed
      (Variables|Default|SetDoingScooterAttack true)
      (Ability|ApplyGameplayEffectToOwner "/ChainsawGameplay/Gameplay/GameplayEffects/GE_TemporarilyBlockSawScooter_Chainsaw.GE_TemporarilyBlockSawScooter_Chainsaw_C")
      (CallFunction|DoRootMotionAimAssist)
      (bind (_asynctaskproxy _notifytag) (Ability|Tasks|PlayMontageAndWaitNotify "None" _montagetoplay _animplayrate "None" false _animrootmotiontranslationscale)
        (:then
          (if (> _returnvalue_2 0)
            (bind _returnvalue_3 (Utilities|Time|SetTimerbyEvent (EventDispatchers|CreateEvent :!event K2_EndAbility) _returnvalue_2))
            (Variables|Default|SetForceEndFailsafeTimer _returnvalue_3)))
        (:OnCompleted
          _)
        (:OnInterrupted
          _)
        (:OnCancelled
          _)
        (:OnNotifyBegin
          (if (== _notifytag "(TagName=\"Animation.Notify.Chainsaw.FinishedScooterAttack\")")
            (Ability|EndAbility)
            (else
              (Development|PrintString "Anim notify" true true "(R=0.000000,G=0.043410,B=1.000000,A=1.000000)")
              (bind _outfortplayerpawn (CallFunction|GetAvatarasValidFortPlayerPawn)
                (:ValidFortPlayerPawn
                  (Ability|ApplyGameplayEffectToOwner "/ChainsawGameplay/Gameplay/GameplayEffects/GE_DoAttackFromSpecial_Chainsaw.GE_DoAttackFromSpecial_Chainsaw_C")
                  (if (and (Variables|Default|GetHasPresetTarget) (not (GameplayTags|HasMatchingGameplayTag (Utilities|Casting|ToGameplayTagAssetInterface (Variables|Default|GetPresetTarget)) "(TagName=\"Granted.Athena.BlockingDamage\")")))
                    (bind _returnvalue_4 (Fortnite|CheckLineOfSightToActorWithChannel (Collision|GetActorBounds (Variables|Default|GetFort%20Player%20Pawn) true false) (Variables|Default|GetPresetTarget) "ECC_GameTraceChannel7" (Variables|Default|GetFort%20Player%20Pawn)))
                    (if _returnvalue_4
                      (GameplayEffects|ApplyGameplayEffectToTarget :self (Ability|GetFortAbilitySystemComponent (Variables|Default|GetFort%20Player%20Pawn)) :GameplayEffectClass "/ChainsawGameplay/Gameplay/GameplayEffects/GE_Chainsaw_MeleeDamage_ScooterAttack.GE_Chainsaw_MeleeDamage_ScooterAttack_C" :Target (Ability|GetFortAbilitySystemComponent (Variables|Default|GetPresetTarget)) :Level 0.0)
                      (CallFunction|ExecuteGameplayCuesonTarget (Ability|TargetData|AbilityTargetDatafromActor (Variables|Default|GetPresetTarget))) ; Play sound / vfx
                    )
                    (else
                      (Utilities|IsValid (Variables|Default|GetTP_SprintAttack)
                        (:Is%20Valid
                          (bind (_asynctaskproxy_2 _targetinghandle) (Class|AsyncActionPerformTargeting|PerformTargetingAsyncAction _outfortplayerpawn (Variables|Default|GetTP_SprintAttack))
                            (:Targeted
                              (if (Utilities|Array|IsNotEmpty (TargetingSystem|TargetingResults|GetTargetingResults "/Script/TargetingSystem.Default__TargetingSubsystem" _targetinghandle))
                                (Ability|CommitAbility) ; Let's do this
                                (for _array_element (TargetingSystem|TargetingResults|GetTargetingResults "/Script/TargetingSystem.Default__TargetingSubsystem" _targetinghandle)
                                  (GameplayEffects|ApplyGameplayEffectToTarget :self (Ability|GetFortAbilitySystemComponent (Variables|Default|GetFort%20Player%20Pawn)) :GameplayEffectClass "/ChainsawGameplay/Gameplay/GameplayEffects/GE_Chainsaw_MeleeDamage_ScooterAttack.GE_Chainsaw_MeleeDamage_ScooterAttack_C" :Target (Ability|GetFortAbilitySystemComponent (.HitActor (Collision|BreakHitResult _array_element))) :Level 0.0)
                                  (CallFunction|ExecuteGameplayCuesonTarget (Ability|TargetData|AbilityTargetDatafromHitResult _array_element)) ; Play sound / vfx
                                  (if (Variables|/Script/Engine.Actor|GetbCanBeDamaged (.HitActor (Collision|BreakHitResult _array_element)))
                                    (ForceFeedback|SpawnForceFeedbackAttached "/ChainsawGameplay/Gameplay/CamShakeAndRumble/FF_HitSomething_Chainsaw.FF_HitSomething_Chainsaw" (Variables|/Script/Engine.Character|GetMesh (Variables|Default|GetFort%20Player%20Pawn)))
                                    (Utilities|IsValid _returnvalue
                                      (:Is%20Valid
                                        (FortWeapon|UseWeaponDurability(Scaled) _returnvalue (ScalableFloat|GetCachedScalableFloatValue (Variables|Default|GetHotfix_ScooterAttackDurabilityCost)))))))
                                (Utilities|FlowControl|Delay (Variables|Default|GetCancelDuration))
                                (else
                                  _))
                              (Ability|EndAbility))) ; Target the attack
                        )
                        (:Is%20Not%20Valid
                          _)) ; Sprint attack targeting preset valid?
                    )))
                (:AvatarCastFailed
                  _)))))) ; Play attack anim
    )))

(event Ability|EventOnEndAbility
  (params (bWasCancelled :type Boolean))
  (bind (_montagetoplay _animplayrate _animrootmotiontranslationscale) (Utilities|Struct|BreakFortGameplayAbilityMontageInfo (Variables|Animation|GetMontageInfo)))
  (bind _returnvalue (Ability|GetCurrentSourceWeapon))
  (bind _sf_forceendfailsafetime (Variables|Default|GetSF_ForceEndFailsafeTime))
  (bind _returnvalue_2 (ScalableFloat|GetValueatLevel _sf_forceendfailsafetime))
  (Utilities|Time|ClearandInvalidateTimerbyHandle (Variables|Default|GetForceEndFailsafeTimer))
  (Variables|Default|SetDoingScooterAttack false)
  ;[1 1 1] Do Attack
  (Utilities|FlowControl|DoOnce false
    (:Completed
      (Variables|Default|SetDoingScooterAttack true)
      (Ability|ApplyGameplayEffectToOwner "/ChainsawGameplay/Gameplay/GameplayEffects/GE_TemporarilyBlockSawScooter_Chainsaw.GE_TemporarilyBlockSawScooter_Chainsaw_C")
      (CallFunction|DoRootMotionAimAssist)
      (bind (_asynctaskproxy _notifytag) (Ability|Tasks|PlayMontageAndWaitNotify "None" _montagetoplay _animplayrate "None" false _animrootmotiontranslationscale)
        (:then
          (if (> _returnvalue_2 0)
            (bind _returnvalue_3 (Utilities|Time|SetTimerbyEvent (EventDispatchers|CreateEvent :!event K2_EndAbility) _returnvalue_2))
            (Variables|Default|SetForceEndFailsafeTimer _returnvalue_3)))
        (:OnCompleted
          _)
        (:OnInterrupted
          _)
        (:OnCancelled
          _)
        (:OnNotifyBegin
          (if (== _notifytag "(TagName=\"Animation.Notify.Chainsaw.FinishedScooterAttack\")")
            (Ability|EndAbility)
            (else
              (Development|PrintString "Anim notify" true true "(R=0.000000,G=0.043410,B=1.000000,A=1.000000)")
              (bind _outfortplayerpawn (CallFunction|GetAvatarasValidFortPlayerPawn)
                (:ValidFortPlayerPawn
                  (Ability|ApplyGameplayEffectToOwner "/ChainsawGameplay/Gameplay/GameplayEffects/GE_DoAttackFromSpecial_Chainsaw.GE_DoAttackFromSpecial_Chainsaw_C")
                  (if (and (Variables|Default|GetHasPresetTarget) (not (GameplayTags|HasMatchingGameplayTag (Utilities|Casting|ToGameplayTagAssetInterface (Variables|Default|GetPresetTarget)) "(TagName=\"Granted.Athena.BlockingDamage\")")))
                    (bind _returnvalue_4 (Fortnite|CheckLineOfSightToActorWithChannel (Collision|GetActorBounds (Variables|Default|GetFort%20Player%20Pawn) true false) (Variables|Default|GetPresetTarget) "ECC_GameTraceChannel7" (Variables|Default|GetFort%20Player%20Pawn)))
                    (if _returnvalue_4
                      (GameplayEffects|ApplyGameplayEffectToTarget :self (Ability|GetFortAbilitySystemComponent (Variables|Default|GetFort%20Player%20Pawn)) :GameplayEffectClass "/ChainsawGameplay/Gameplay/GameplayEffects/GE_Chainsaw_MeleeDamage_ScooterAttack.GE_Chainsaw_MeleeDamage_ScooterAttack_C" :Target (Ability|GetFortAbilitySystemComponent (Variables|Default|GetPresetTarget)) :Level 0.0)
                      (CallFunction|ExecuteGameplayCuesonTarget (Ability|TargetData|AbilityTargetDatafromActor (Variables|Default|GetPresetTarget))) ; Play sound / vfx
                    )
                    (else
                      (Utilities|IsValid (Variables|Default|GetTP_SprintAttack)
                        (:Is%20Valid
                          (bind (_asynctaskproxy_2 _targetinghandle) (Class|AsyncActionPerformTargeting|PerformTargetingAsyncAction _outfortplayerpawn (Variables|Default|GetTP_SprintAttack))
                            (:Targeted
                              (if (Utilities|Array|IsNotEmpty (TargetingSystem|TargetingResults|GetTargetingResults "/Script/TargetingSystem.Default__TargetingSubsystem" _targetinghandle))
                                (Ability|CommitAbility) ; Let's do this
                                (for _array_element (TargetingSystem|TargetingResults|GetTargetingResults "/Script/TargetingSystem.Default__TargetingSubsystem" _targetinghandle)
                                  (GameplayEffects|ApplyGameplayEffectToTarget :self (Ability|GetFortAbilitySystemComponent (Variables|Default|GetFort%20Player%20Pawn)) :GameplayEffectClass "/ChainsawGameplay/Gameplay/GameplayEffects/GE_Chainsaw_MeleeDamage_ScooterAttack.GE_Chainsaw_MeleeDamage_ScooterAttack_C" :Target (Ability|GetFortAbilitySystemComponent (.HitActor (Collision|BreakHitResult _array_element))) :Level 0.0)
                                  (CallFunction|ExecuteGameplayCuesonTarget (Ability|TargetData|AbilityTargetDatafromHitResult _array_element)) ; Play sound / vfx
                                  (if (Variables|/Script/Engine.Actor|GetbCanBeDamaged (.HitActor (Collision|BreakHitResult _array_element)))
                                    (ForceFeedback|SpawnForceFeedbackAttached "/ChainsawGameplay/Gameplay/CamShakeAndRumble/FF_HitSomething_Chainsaw.FF_HitSomething_Chainsaw" (Variables|/Script/Engine.Character|GetMesh (Variables|Default|GetFort%20Player%20Pawn)))
                                    (Utilities|IsValid _returnvalue
                                      (:Is%20Valid
                                        (FortWeapon|UseWeaponDurability(Scaled) _returnvalue (ScalableFloat|GetCachedScalableFloatValue (Variables|Default|GetHotfix_ScooterAttackDurabilityCost)))))))
                                (Utilities|FlowControl|Delay (Variables|Default|GetCancelDuration))
                                (else
                                  _))
                              (Ability|EndAbility))) ; Target the attack
                        )
                        (:Is%20Not%20Valid
                          _)) ; Sprint attack targeting preset valid?
                    )))
                (:AvatarCastFailed
                  _)))))) ; Play attack anim
    )))

(event Custom|DoRootMotionAimAssist
  (bind (_bblockinghit _bwalkablefloor _blinetrace _floordist _linedist _hitresult) (Utilities|Struct|BreakFindFloorResult (Pawn|Components|CharacterMovement|FindFloor (Variables|/Script/Engine.Character|GetCharacterMovement (Variables|Default|GetFort%20Player%20Pawn)) (Transformation|GetWorldLocation (Variables|/Script/Engine.Character|GetCapsuleComponent (Variables|Default|GetFort%20Player%20Pawn))))))
  (bind (_bblockinghit_2 _binitialoverlap _time _distance _location _impactpoint _normal _impactnormal) (Collision|BreakHitResult _hitresult))
  (if (Variables|Default|GetHasPresetTarget)
    (Utilities|ApplyMovetoTargetRootMotion (Variables|Default|GetPresetTarget))
    (else
      (bind (_asynctaskproxy _targetinghandle) (Class|AsyncActionPerformTargeting|PerformTargetingAsyncAction (Variables|Default|GetFort%20Player%20Pawn) "/ChainsawGameplay/Gameplay/TP_Chainsaw_RootMotionScooterAttack_AimAssist.TP_Chainsaw_RootMotionScooterAttack_AimAssist")
        (:Targeted
          (bind (_foundaplayer _return) (CallFunction|GetFirstPlayerTarget (TargetingSystem|TargetingResults|GetTargetingResults "/Script/TargetingSystem.Default__TargetingSubsystem" _targetinghandle)))
          (if _foundaplayer
            (Utilities|ApplyMovetoTargetRootMotion _return)
            (else
              (Ability|Tasks|ApplyRootMotion_Slide "None" 0.7 (select (and _bblockinghit (ScalableFloat|GetValueasBool (Variables|Default|GetSF_AllignToFloor))) (Math|Vector|Normalize (Math|Vector|ProjectVectorontoPlane (Transformation|GetActorForwardVector (Variables|Default|GetFort%20Player%20Pawn)) (Math|Vector|Normalize (Math|Vector|ProjectVectorontoPlane _impactnormal (Transformation|GetActorRightVector (Variables|Default|GetFort%20Player%20Pawn)))))) (Transformation|GetActorForwardVector (Variables|Default|GetFort%20Player%20Pawn))) 1.2e+03 1.2e+03 0 0.0 "/ChainsawGameplay/Gameplay/Curve_JumpAttack_Chainsaw_PathOffsetCurve.Curve_JumpAttack_Chainsaw_PathOffsetCurve" "MaintainLastRootMotionVelocity" "0, 0, 0" 0.0))))) ; Target the attack
    )))

(fn ExecuteGameplayCuesOnTarget
  (params (TargetData :type Gameplay%20Ability%20Target%20Data%20Handle%20Structure))
  (for _array_element (Ability|TargetData|GetAllActorsfromTargetData (Ability|TargetData|FilterTargetData (Variables|GetTargetData) (Filter|MakeFilterHandle (Utilities|Struct|MakeGameplayTargetDataFilter "/Script/FortniteGame.FortPawn" "TDFS_NoSelf") (Ability|GetAvatarActorfromActorInfo))))
    (Ability|ExecuteGameplayCueWithParamsOnOwner (Variables|Default|GetGCN_EnemyHit) (Ability|GameplayCue|MakeGameplayCueParameters :NormalizedMagnitude 0.0 :RawMagnitude 0.0 :Location (Transformation|GetActorLocation _array_element) :Normal "0,0,0" :GameplayEffectLevel 1 :AbilityLevel 1 :bReplicateLocationWhenUsingMinimalRepProxy false)))
  (for _array_element_2 (Ability|TargetData|GetAllActorsfromTargetData (Ability|TargetData|FilterTargetData (Variables|GetTargetData) (Filter|MakeFilterHandle (Utilities|Struct|MakeGameplayTargetDataFilter "/Script/FortniteGame.FortPawn" "TDFS_Any" true) (Ability|GetAvatarActorfromActorInfo))))
    (Ability|ExecuteGameplayCueWithParamsOnOwner (Variables|Default|GetGCN_EnvironmentHit) (Ability|GameplayCue|MakeGameplayCueParameters :NormalizedMagnitude 0.0 :RawMagnitude 0.0 :Location (Transformation|GetActorLocation _array_element_2) :Normal "0,0,0" :GameplayEffectLevel 1 :AbilityLevel 1 :bReplicateLocationWhenUsingMinimalRepProxy false))))

(fn GetFirstPlayerTarget
  (params (Array :type Array%20of%20Hit%20Result%20Structures))
  (for _array_element Array
    (Development|PrintString (Utilities|GetDisplayName (.HitActor (Collision|BreakHitResult _array_element))) true true "(R=1.000000,G=0.044127,B=0.000000,A=1.000000)" 5.0)
    (bind _aspawn (Utilities|Casting|CastToPawn (.HitActor (Collision|BreakHitResult _array_element))))
    (return true _aspawn))
  (return false 0))
