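; Secondary chainsaw gameplay ability that triggers the saw scooter while input is held, failing when Overheat exceeds 99.
; Activation removes balloons; if falling or ballooning it applies a FallingIntoScooter effect, otherwise a forward sphere trace picks between auto-attacking a player and starting the scooter through gameplay events.
; Release ends the ability, removing that effect and sending the stop event.
; An override silences the failure message when BlockScootering blocks activation.
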(event Ability|EventActivateAbility
  (bind _returnvalue (CallFunction|GetAvatarasFortPlayerPawn))
  (bind (_bsuccessfullyfoundattribute _returnvalue_2) (Ability|Attribute|GetFloatAttribute (CallFunction|GetAvatarasFortPlayerPawn) "(AttributeName=\"Overheat\",Attribute=/Script/ChainsawGameplayRuntime.ChainsawAttributeSet:Overheat,AttributeOwner=\"/Script/CoreUObject.Class'/Script/ChainsawGameplayRuntime.ChainsawAttributeSet'\")"))
  (|CallParentFunction)
  (if (<= _returnvalue_2 99.0)
    ;[1 0.3955 0] Interaction Matrix: Cancel conflicting abilities on ability start
    (Development|PrintString "ACTIVATE")
    (Ability|SendGameplayEvent (Variables|Default|GetEventTag_RemoveBalloons)) ; remove all balloons from the player \nsince they are doing a MME action\nand should be consistent with other features\n(like Doom Gauntlets - FORT-766407)
    (if (GameplayTags|HasAnyMatchingGameplayTags (CallFunction|GetAvatarasFortPlayerPawn) (GameplayTags|MakeLiteralGameplayTagContainer "(GameplayTags=((TagName=\"MovementMode.Falling\"),(TagName=\"MovementMode.Ballooning\")))"))
      (CallFunction|PlayScooterJumpStart)
      (Ability|ApplyGameplayEffectToOwner "/ChainsawGameplay/Gameplay/GameplayEffects/GE_FallingIntoScooter_Chainsaw.GE_FallingIntoScooter_Chainsaw_C")
      (else
        (bind _outhits (Collision|MultiSphereTraceByProfile (+ (Transformation|GetActorLocation (CallFunction|GetAvatarasFortPlayerPawn)) (* (Math|Vector|GetForwardVector (Pawn|GetDesiredRotation (Pawn|GetController (CallFunction|GetAvatarasFortPlayerPawn)))) (* (Variables|/Script/Engine.CapsuleComponent|GetCapsuleRadius (Variables|/Script/Engine.Character|GetCapsuleComponent (CallFunction|GetAvatarasFortPlayerPawn))) (ScalableFloat|GetValueatLevel (Variables|Default|GetSF_AutoAttack_TriggerTrace_StartOffset))))) (+ (Transformation|GetActorLocation (CallFunction|GetAvatarasFortPlayerPawn)) (* (Math|Vector|GetForwardVector (Pawn|GetDesiredRotation (Pawn|GetController (CallFunction|GetAvatarasFortPlayerPawn)))) (ScalableFloat|GetValueatLevel (Variables|Default|GetSF_AutoAttack_TriggerTrace_Distance)))) (* (Variables|/Script/Engine.CapsuleComponent|GetCapsuleRadius (Variables|/Script/Engine.Character|GetCapsuleComponent (CallFunction|GetAvatarasFortPlayerPawn))) (ScalableFloat|GetValueatLevel (Variables|Default|GetSF_AutoAttack_TriggerTrace_Radius))) "FortTriggerOnlyPawns" false (Utilities|Array|MakeArray (CallFunction|GetAvatarasFortPlayerPawn)) "None" true "(R=1.000000,G=0.000000,B=0.850735,A=1.000000)" "(R=0.000000,G=1.000000,B=0.851418,A=1.000000)" 0.2)) ; Larger trace for players
        (bind (_foundaplayer _return) (CallFunction|GetFirstPlayerTarget _outhits))
        (if (Math|Boolean|ANDBoolean _foundaplayer (GameplayTags|HasMatchingGameplayTag _returnvalue "(TagName=\"Granted.Athena.Chainsaw.ShouldDoAutoScooterAttack\")") (Class|AthenaFunctionLibrary|RowtoBool (Variables|Default|GetSF_AutoAttack_EnabledFromStart)))
          (Ability|SendGameplayEventtoActor (CallFunction|GetAvatarasFortPlayerPawn) "(TagName=\"Event.Athena.Chainsaw.DoScooterAttack\")" (Utilities|Struct|MakeGameplayEventData "(TagName=\"Event.Athena.Chainsaw.DoScooterAttack\")" 0 _return))
          (elif (GameplayTags|HasMatchingGameplayTag _returnvalue "(TagName=\"Granted.Athena.Chainsaw.FallingIntoScooter\")")
            (Ability|SendGameplayEventtoActor (CallFunction|GetAvatarasFortPlayerPawn) "(TagName=\"Event.Athena.Chainsaw.StartSawScooter.FromAir\")")
            (else
              (Ability|SendGameplayEventtoActor (CallFunction|GetAvatarasFortPlayerPawn) "(TagName=\"Event.Athena.Chainsaw.StartSawScooter\")")
              (CallFunction|PlayScooterJumpStart))))))
    (else
      (FailedAction|UnabletoPerformAction (CallFunction|GetAvatarasFortPlayerPawn) true (GameplayTags|MakeLiteralGameplayTagContainer "(GameplayTags=((TagName=\"Athena.Ability.Failed.Chainsaw.Overheating\")))")))))

(event Ability|EventOnEndAbility
  (params (bWasCancelled :type Boolean))
  (Ability|RemoveGameplayEffectFromOwnerWithGrantedTags "(GameplayTags=((TagName=\"Granted.Athena.Chainsaw.FallingIntoScooter\")))")
  (Ability|SendGameplayEventtoActor (CallFunction|GetAvatarasFortPlayerPawn) "(TagName=\"Event.Athena.Chainsaw.StopSawScooter\")")
  (Development|PrintString "STOP SCOOTER"))

(event Ability|EventOnAbilityInputReleased
  (|CallParentFunction)
  (Ability|EndAbility))

(event Custom|PlayScooterJumpStart
  (Ability|ExecuteGameplayCueOnOwner (Variables|Default|GetGC_StartScooterSound)))

(fn GetFirstPlayerTarget
  (params (Array :type Array%20of%20Hit%20Result%20Structures))
  (for _array_element Array
    (bind _asplayer_pawn_athena (Utilities|Casting|CastToPlayerPawn_Athena (.HitActor (Collision|BreakHitResult _array_element))))
    (return true _asplayer_pawn_athena))
  (return false 0))

(fn OverrideFailedReason
  (params (FailedReason :type Gameplay%20Tag%20Container%20Structure%20%28by%20ref%29))
  (if (GameplayTags|HasAllTags FailedReason (GameplayTags|MakeLiteralGameplayTagContainer "(GameplayTags=((TagName=\"Activate.Failed.Tags.Blocked\"),(TagName=\"Granted.Athena.Chainsaw.BlockScootering\")))") true)
    (return "(GameplayTags=((TagName=\"Athena.Ability.Failed.NoMessage\")))")
    (else
      (return FailedReason))))
