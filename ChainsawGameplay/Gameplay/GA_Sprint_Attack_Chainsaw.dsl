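; Gameplay ability for the chainsaw sprint attack, enabled by hotfix and fired when tactical sprint ends (unless hurdling) or the ability ends.
; The attack runs once, playing a montage whose notify triggers preset targeting, then damage, hit cues and rumble.
; An aim-assist pass uses root motion to pull toward the first pawn found, otherwise slides forward, optionally floor-aligned, with CVar-gated velocity clamping.
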(event Ability|EventActivateAbility
  (bind _fort_player_pawn (Variables|Default|GetFort%20Player%20Pawn))
  (bind _returnvalue (Pawn|GetController _fort_player_pawn))
  (bind _returnvalue_2 (Actor|GetComponentbyClass _returnvalue "/TacticalSprintGame/Gameplay/TacticalSprintControllerComponent.TacticalSprintControllerComponent_C"))
  ;[1 1 1] Activate
  (if (ScalableFloat|GetValueasBool (Variables|Default|GetSprintAttackEnabledHotfixRow))
    (bind _outfortplayerpawn (CallFunction|GetAvatarasValidFortPlayerPawn)
      (:ValidFortPlayerPawn
        (Variables|Default|SetFort%20Player%20Pawn _outfortplayerpawn)
        (Ability|ApplyGameplayEffectToOwner "/ChainsawGameplay/Gameplay/GameplayEffects/GE_ApplySoundIndicatorCue.GE_ApplySoundIndicatorCue_C")
        (Utilities|IsValid _returnvalue_2
          (:Is%20Valid
            (bind _output_get (Variables|Default|SetTacticalSprintComponent _returnvalue_2))
            (Sprint|AssignOnTacticalSprintEnded _output_get (EventDispatchers|CreateEvent :!event Custom|OnTacticalSprintEnded_Event)) ; Listen for sprint ended
          )
          (:Is%20Not%20Valid
            _)) ; Does this pawn have tac sprint?
      )
      (:AvatarCastFailed
        _))
    (Ability|EndAbility)
    (else
      (Ability|EndAbility))))

(event Custom|OnTacticalSprintEnded_Event
  (bind (_montagetoplay _animplayrate _animrootmotiontranslationscale) (Utilities|Struct|BreakFortGameplayAbilityMontageInfo (Variables|Animation|GetMontageInfo)))
  ;[1 1 1] Do Attack
  (if (GameplayTags|HasAnyMatchingGameplayTags (Variables|Default|GetFort%20Player%20Pawn) (GameplayTags|MakeLiteralGameplayTagContainer "(GameplayTags=((TagName=\"Gameplay.Action.Player.Hurdle.Hurdling\")))"))
    (Ability|EndAbility)
    (else
      (Utilities|FlowControl|DoOnce false
        (:Completed
          (CallFunction|DoRootMotionAimAssist)
          (GameplayCue|ExecuteGameplayCue%28Burst%29Local (Variables|Default|GetFort%20Player%20Pawn) (Variables|Default|GetGC_OnAttack) (Ability|GameplayCue|MakeGameplayCueParameters 0.0))
          (Ability|Tasks|PlayMontageAndWaitNotify "None" _montagetoplay _animplayrate "None" false _animrootmotiontranslationscale
            (:OnCompleted
              _)
            (:OnInterrupted
              _)
            (:OnCancelled
              _)
            (:OnNotifyBegin
              (bind _outfortplayerpawn (CallFunction|GetAvatarasValidFortPlayerPawn)
                (:ValidFortPlayerPawn
                  (Utilities|IsValid (Variables|Default|GetTP_SprintAttack)
                    (:Is%20Valid
                      (Ability|ApplyGameplayEffectToOwner "/ChainsawGameplay/Gameplay/GameplayEffects/GE_DoAttackFromSpecial_Chainsaw.GE_DoAttackFromSpecial_Chainsaw_C")
                      (bind (_asynctaskproxy _targetinghandle) (Class|AsyncActionPerformTargeting|PerformTargetingAsyncAction _outfortplayerpawn (Variables|Default|GetTP_SprintAttack))
                        (:Targeted
                          (if (Utilities|Array|IsNotEmpty (TargetingSystem|TargetingResults|GetTargetingResults "/Script/TargetingSystem.Default__TargetingSubsystem" _targetinghandle))
                            (Ability|CommitAbility) ; Let's do this
                            (for _array_element (TargetingSystem|TargetingResults|GetTargetingResults "/Script/TargetingSystem.Default__TargetingSubsystem" _targetinghandle)
                              (GameplayEffects|ApplyGameplayEffectToTarget :self (Ability|GetFortAbilitySystemComponent (Variables|Default|GetFort%20Player%20Pawn)) :GameplayEffectClass "/ChainsawGameplay/Gameplay/GameplayEffects/GE_Chainsaw_MeleeDamage_SprintAttack.GE_Chainsaw_MeleeDamage_SprintAttack_C" :Target (Ability|GetFortAbilitySystemComponent (.HitActor (Collision|BreakHitResult _array_element))) :Level 0.0)
                              (CallFunction|ExecuteGameplayCuesonTarget (Ability|TargetData|AbilityTargetDatafromHitResult _array_element)) ; Play sound / vfx
                              (Utilities|IsValid (Ability|GetFortAbilitySystemComponent (.HitActor (Collision|BreakHitResult _array_element)))
                                (:Is%20Valid
                                  (ForceFeedback|SpawnForceFeedbackAttached "/ChainsawGameplay/Gameplay/CamShakeAndRumble/FF_HitSomething_Chainsaw.FF_HitSomething_Chainsaw" (Variables|/Script/Engine.Character|GetMesh (Variables|Default|GetFort%20Player%20Pawn))))))
                            (Utilities|FlowControl|Delay (Variables|Default|GetCancelDuration))
                            (else
                              _))
                          (Ability|EndAbility))) ; Target the attack
                    )
                    (:Is%20Not%20Valid
                      _)) ; Sprint attack targeting preset valid?
                )
                (:AvatarCastFailed
                  _)))) ; Play attack anim
          (Ability|EndAbility))))))

(event Ability|EventOnEndAbility
  (params (bWasCancelled :type Boolean))
  (bind (_montagetoplay _animplayrate _animrootmotiontranslationscale) (Utilities|Struct|BreakFortGameplayAbilityMontageInfo (Variables|Animation|GetMontageInfo)))
  (bind _tacticalsprintcomponent (Variables|Default|GetTacticalSprintComponent))
  ;[1 1 1] End Ability
  (Utilities|FlowControl|Sequence
    (:then_0
      ;[1 1 1] Do Attack
      (Utilities|FlowControl|DoOnce false
        (:Completed
          (CallFunction|DoRootMotionAimAssist)
          (GameplayCue|ExecuteGameplayCue%28Burst%29Local (Variables|Default|GetFort%20Player%20Pawn) (Variables|Default|GetGC_OnAttack) (Ability|GameplayCue|MakeGameplayCueParameters 0.0))
          (Ability|Tasks|PlayMontageAndWaitNotify "None" _montagetoplay _animplayrate "None" false _animrootmotiontranslationscale
            (:OnCompleted
              _)
            (:OnInterrupted
              _)
            (:OnCancelled
              _)
            (:OnNotifyBegin
              (bind _outfortplayerpawn (CallFunction|GetAvatarasValidFortPlayerPawn)
                (:ValidFortPlayerPawn
                  (Utilities|IsValid (Variables|Default|GetTP_SprintAttack)
                    (:Is%20Valid
                      (Ability|ApplyGameplayEffectToOwner "/ChainsawGameplay/Gameplay/GameplayEffects/GE_DoAttackFromSpecial_Chainsaw.GE_DoAttackFromSpecial_Chainsaw_C")
                      (bind (_asynctaskproxy _targetinghandle) (Class|AsyncActionPerformTargeting|PerformTargetingAsyncAction _outfortplayerpawn (Variables|Default|GetTP_SprintAttack))
                        (:Targeted
                          (if (Utilities|Array|IsNotEmpty (TargetingSystem|TargetingResults|GetTargetingResults "/Script/TargetingSystem.Default__TargetingSubsystem" _targetinghandle))
                            (Ability|CommitAbility) ; Let's do this
                            (for _array_element (TargetingSystem|TargetingResults|GetTargetingResults "/Script/TargetingSystem.Default__TargetingSubsystem" _targetinghandle)
                              (GameplayEffects|ApplyGameplayEffectToTarget :self (Ability|GetFortAbilitySystemComponent (Variables|Default|GetFort%20Player%20Pawn)) :GameplayEffectClass "/ChainsawGameplay/Gameplay/GameplayEffects/GE_Chainsaw_MeleeDamage_SprintAttack.GE_Chainsaw_MeleeDamage_SprintAttack_C" :Target (Ability|GetFortAbilitySystemComponent (.HitActor (Collision|BreakHitResult _array_element))) :Level 0.0)
                              (CallFunction|ExecuteGameplayCuesonTarget (Ability|TargetData|AbilityTargetDatafromHitResult _array_element)) ; Play sound / vfx
                              (Utilities|IsValid (Ability|GetFortAbilitySystemComponent (.HitActor (Collision|BreakHitResult _array_element)))
                                (:Is%20Valid
                                  (ForceFeedback|SpawnForceFeedbackAttached "/ChainsawGameplay/Gameplay/CamShakeAndRumble/FF_HitSomething_Chainsaw.FF_HitSomething_Chainsaw" (Variables|/Script/Engine.Character|GetMesh (Variables|Default|GetFort%20Player%20Pawn))))))
                            (Utilities|FlowControl|Delay (Variables|Default|GetCancelDuration))
                            (else
                              _))
                          (Ability|EndAbility))) ; Target the attack
                    )
                    (:Is%20Not%20Valid
                      _)) ; Sprint attack targeting preset valid?
                )
                (:AvatarCastFailed
                  _)))) ; Play attack anim
          (Ability|EndAbility))))
    (:then_1
      (Utilities|IsValid _tacticalsprintcomponent
        (:Is%20Valid
          (Sprint|UnbindEventfromOnTacticalSprintEnded _tacticalsprintcomponent (EventDispatchers|CreateEvent :!event Custom|OnTacticalSprintEnded_Event)))))))

(event Custom|DoRootMotionAimAssist
  (bind _fort_player_pawn (Variables|Default|GetFort%20Player%20Pawn))
  (bind _capsulecomponent (Variables|/Script/Engine.Character|GetCapsuleComponent _fort_player_pawn))
  (bind _returnvalue (Transformation|GetWorldLocation _capsulecomponent))
  (bind _charactermovement (Variables|/Script/Engine.Character|GetCharacterMovement _fort_player_pawn))
  (bind _floorresult (Pawn|Components|CharacterMovement|FindFloor _charactermovement _returnvalue))
  (bind (_bblockinghit _bwalkablefloor _blinetrace _floordist _linedist _hitresult) (Utilities|Struct|BreakFindFloorResult _floorresult))
  (bind (_bblockinghit_2 _binitialoverlap _time _distance _location _impactpoint _normal _impactnormal) (Collision|BreakHitResult _hitresult))
  (bind _returnvalue_2 (Math|Vector|ProjectVectorontoPlane _impactnormal (Transformation|GetActorRightVector _fort_player_pawn)))
  (bind _returnvalue_3 (Math|Vector|Normalize _returnvalue_2))
  (bind _returnvalue_4 (Transformation|GetActorForwardVector _fort_player_pawn))
  (bind _returnvalue_5 (Math|Vector|ProjectVectorontoPlane _returnvalue_4 _returnvalue_3))
  (bind _returnvalue_6 (Math|Vector|Normalize _returnvalue_5))
  (bind _returnvalue_7 (Transformation|GetActorRightVector _fort_player_pawn))
  (bind _sf_alligntofloor (Variables|Default|GetSF_AllignToFloor))
  (bind _returnvalue_8 (ScalableFloat|GetValueasBool _sf_alligntofloor))
  (bind _returnvalue_9 (and _bblockinghit _returnvalue_8))
  (bind _returnvalue_10 (select _returnvalue_9 _returnvalue_6 _returnvalue_4))
  (bind (_asynctaskproxy _targetinghandle) (Class|AsyncActionPerformTargeting|PerformTargetingAsyncAction (Variables|Default|GetFort%20Player%20Pawn) "/ChainsawGameplay/Gameplay/TP_Chainsaw_SprintAttack_AimAssist.TP_Chainsaw_SprintAttack_AimAssist")
    (:Targeted
      (bind (_foundaplayer _return) (CallFunction|GetFirstPlayerTarget (TargetingSystem|TargetingResults|GetTargetingResults "/Script/TargetingSystem.Default__TargetingSubsystem" _targetinghandle)))
      (if _foundaplayer
        (Ability|Tasks|ApplyRootMotionMoveToActorForce "None" _return "0, 0, 0" "AlignFromTargetToSource" 0.7 0 0 false "MOVE_None" false "/ChainsawGameplay/Gameplay/Curve_JumpAttack_Chainsaw_PathOffsetCurve.Curve_JumpAttack_Chainsaw_PathOffsetCurve" 0 "MaintainLastRootMotionVelocity" "0, 0, 0" 0.0 false 1.5e+02)
        (else
          (bind _returnvalue_11 (Development|GetConsoleVariableBoolValue "Fort.Rollback.ClampChainsawZValue"))
          (if _returnvalue_11
            (Ability|Tasks|ApplyRootMotion_Slide "None" 0.7 _returnvalue_10 1.2e+03 1.2e+03 0 0.0 "/ChainsawGameplay/Gameplay/Curve_JumpAttack_Chainsaw_PathOffsetCurve.Curve_JumpAttack_Chainsaw_PathOffsetCurve" "ClampVelocity" "0, 0, 0" (ScalableFloat|GetValueatLevel (Variables|Default|GetClamp%20Velocity)))
            (else
              (Ability|Tasks|ApplyRootMotion_Slide "None" 0.7 _returnvalue_10 1.2e+03 1.2e+03 0 0.0 "/ChainsawGameplay/Gameplay/Curve_JumpAttack_Chainsaw_PathOffsetCurve.Curve_JumpAttack_Chainsaw_PathOffsetCurve" "MaintainLastRootMotionVelocity" "0, 0, 0" 0.0))))))) ; Target the attack
)

(fn ExecuteGameplayCuesOnTarget
  (params (TargetData :type Gameplay%20Ability%20Target%20Data%20Handle%20Structure))
  (for _array_element (Ability|TargetData|GetAllActorsfromTargetData (Ability|TargetData|FilterTargetData (Variables|GetTargetData) (Filter|MakeFilterHandle (Utilities|Struct|MakeGameplayTargetDataFilter "/Script/FortniteGame.FortPawn" "TDFS_NoSelf") (Ability|GetAvatarActorfromActorInfo))))
    (Ability|ExecuteGameplayCueWithParamsOnOwner (Variables|Default|GetGCN_EnemyHit) (Ability|GameplayCue|MakeGameplayCueParameters :NormalizedMagnitude 0.0 :RawMagnitude 0.0 :Location (Transformation|GetActorLocation _array_element) :Normal "0,0,0" :GameplayEffectLevel 1 :AbilityLevel 1 :bReplicateLocationWhenUsingMinimalRepProxy false)))
  (for _array_element_2 (Ability|TargetData|GetAllActorsfromTargetData (Ability|TargetData|FilterTargetData (Variables|GetTargetData) (Filter|MakeFilterHandle (Utilities|Struct|MakeGameplayTargetDataFilter "/Script/FortniteGame.FortPawn" "TDFS_Any" true) (Ability|GetAvatarActorfromActorInfo))))
    (Ability|ExecuteGameplayCueWithParamsOnOwner (Variables|Default|GetGCN_EnvironmentHit) (Ability|GameplayCue|MakeGameplayCueParameters :NormalizedMagnitude 0.0 :RawMagnitude 0.0 :Location (Transformation|GetActorLocation _array_element_2) :Normal "0,0,0" :GameplayEffectLevel 1 :AbilityLevel 1 :bReplicateLocationWhenUsingMinimalRepProxy false))))

(fn GetFirstPlayerTarget
  (params (Array :type Array%20of%20Hit%20Result%20Structures))
  (for _array_element Array
    (bind _aspawn (Utilities|Casting|CastToPawn (.HitActor (Collision|BreakHitResult _array_element))))
    (return true _aspawn))
  (return false 0))
